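# Settings-row widget, built on CommonUserWidget, that shows a header and a daily time that steps by a configurable minute increment.
# Init seeds time and increment; arrow buttons or gamepad left/right adjust it, refresh the text and broadcast OnTimeChanged.
# Panel_Value's enabled state defines editability, selecting the hover animation.
# Input type, hover and editability drive arrow visibility through animations, while compact UI swaps outline and arrow brushes.
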WidgetBlueprint /FortniteCoreUI/Settings/WBP_DateTimeSelector
animations [
 OnHover {
  binding_count=2
  length_seconds=0.1
  object=&self:OnHover
  sequence {
   bindings [
    FocusOutline {
     id=#1
     possessed_class=Image
     tracks [{
       type=MovieSceneFloatTrack
       display_name=Render Opacity
       sections [{type=MovieSceneFloatSection, end_frame=12, start_frame=0, channels [None {type=MovieSceneScriptingFloatChannel, keys [[0,0],[12,1]]}]}]
      }]
    }
    Text_SettingName {
     id=#2
     possessed_class=CommonTextBlock
     tracks [{
       type=MovieSceneColorTrack
       display_name=Color and Opacity
       sections [{
         type=MovieSceneColorSection
         end_frame=12
         start_frame=0
         channels [
          Color.R {type=MovieSceneScriptingFloatChannel, keys [[0,0.521],[12,1]]}
          Color.G {type=MovieSceneScriptingFloatChannel, keys [[0,0.521],[12,1]]}
          Color.B {type=MovieSceneScriptingFloatChannel, keys [[0,0.521],[12,1]]}
          Color.A {type=MovieSceneScriptingFloatChannel, keys [[0,0.8],[12,1]]}
         ]
        }]
      }]
    }
   ]
   display_rate {numerator=120}
  }
 }
 OnHoverArrows {
  binding_count=2
  length_seconds=0.1
  object=&self:OnHoverArrows
  sequence {
   bindings [
    Button_Increase {
     id=#3
     possessed_class=IconTextButton_C
     tracks [
      {
       type=MovieSceneColorTrack
       display_name=Color and Opacity
       sections [{
         type=MovieSceneColorSection
         end_frame=12
         start_frame=0
         channels [
          Color.R {type=MovieSceneScriptingFloatChannel}
          Color.G {type=MovieSceneScriptingFloatChannel}
          Color.B {type=MovieSceneScriptingFloatChannel}
          Color.A {type=MovieSceneScriptingFloatChannel, keys [[0,0],[12,1]]}
         ]
        }]
      }
      {
       type=MovieSceneByteTrack
       display_name=Visibility
       sections [{type=MovieSceneByteSection, end_frame=12, start_frame=0, channels [None {type=MovieSceneScriptingByteChannel, keys [[0,4],[12,0]]}]}]
      }
     ]
    }
    Button_Decrease {
     id=#4
     possessed_class=IconTextButton_C
     tracks [
      {
       type=MovieSceneColorTrack
       display_name=Color and Opacity
       sections [{
         type=MovieSceneColorSection
         end_frame=12
         start_frame=0
         channels [
          Color.R {type=MovieSceneScriptingFloatChannel}
          Color.G {type=MovieSceneScriptingFloatChannel}
          Color.B {type=MovieSceneScriptingFloatChannel}
          Color.A {type=MovieSceneScriptingFloatChannel, keys [[0,0],[12,1]]}
         ]
        }]
      }
      {
       type=MovieSceneByteTrack
       display_name=Visibility
       sections [{type=MovieSceneByteSection, end_frame=12, start_frame=0, channels [None {type=MovieSceneScriptingByteChannel, keys [[0,4],[12,0]]}]}]
      }
     ]
    }
   ]
   display_rate {numerator=120}
  }
 }
 OnSelect_Touch {
  binding_count=1
  length_seconds=0.5
  object=&self:OnSelect_Touch
  sequence {bindings [Background {
     id=#5
     possessed_class=SettingEntryBackground_C
     tracks [{
       type=MovieSceneColorTrack
       display_name=Color and Opacity
       sections [{
         type=MovieSceneColorSection
         end_frame=10
         start_frame=0
         channels [
          Color.R {type=MovieSceneScriptingFloatChannel, const [1,2]}
          Color.G {type=MovieSceneScriptingFloatChannel, keys [[0,1],[10,0]]}
          Color.B {type=MovieSceneScriptingFloatChannel, keys [[0,1],[10,0]]}
          Color.A {type=MovieSceneScriptingFloatChannel, const [1,1]}
         ]
        }]
      }]
    }]}
 }
 OnHoverDisabled {
  binding_count=1
  length_seconds=0.1
  object=&self:OnHoverDisabled
  sequence {
   bindings [FocusOutline {
     id=#1
     possessed_class=Image
     tracks [{
       type=MovieSceneFloatTrack
       display_name=Render Opacity
       sections [{type=MovieSceneFloatSection, end_frame=12, start_frame=0, channels [None {type=MovieSceneScriptingFloatChannel, keys [[0,0],[12,1]]}]}]
      }]
    }]
   display_rate {numerator=120}
  }
 }
]
graphs [
 Change Touch Selection State {nodes [
   n1 = Change Touch Selection State!() -> (then, IsSelected:bool=~) {is=title K2Node_FunctionEntry}
   > n2 = Branch(Condition=@n1)
   n3 = Get bIsInSelectedState() -> (bIsInSelectedState:bool)
   > n4 = Branch(Condition=@n3)
   Set bIsInSelectedState(execute=@n4.else, bIsInSelectedState:bool=true) -> (Output_Get:bool)
   n5 = GetOnHoverAnimation!() -> (OnHoverAnimation:Widget Animation Object Reference)
   > PlayAnimationForward(InAnimation=@n5)
   > SpawnSound2D!(
    execute
    Sound:Sound Base Object Reference=Get HoverSound() -> (HoverSound:Sound Base Object Reference)
    VolumeMultiplier:float=1
    PitchMultiplier:float=1
    StartTime:float=0
    ConcurrencySettings:Sound Concurrency Object Reference
    bPersistAcrossLevelTransition:bool=false
    bAutoDestroy:bool=true
   ) -> (then, ReturnValue:Audio Component Object Reference)
   Branch(execute=@n2.else, Condition=@n3)
   > Set bIsInSelectedState(bIsInSelectedState=false)
   > PlayAnimationReverse(InAnimation=@n5, PlaybackSpeed=2)
  ]}
 DecrementTime {nodes [
   DecrementTime() {is=title K2Node_FunctionEntry}
   n1 = Get TimeIncrementMinutes() -> (TimeIncrementMinutes:int=0)
   > Set Time(Time:Daily Time Structure=DailyTime - DailyTime(
     A:Daily Time Structure=Get Time() -> (Time:Daily Time Structure)
     B:Daily Time Structure=MakeDailyTime!(Hour:int=0, Minute:int=@n1) -> (ReturnValue:Daily Time Structure)
    ) -> (ReturnValue:Daily Time Structure) {is=title K2Node_PromotableOperator}) -> (Output_Get:Daily Time Structure)
   > RefreshUI()
   > Call On Time Changed!(execute:Exec, self:WBP Date Time Selector Object Reference, NewTime:Daily Time Structure=Get Time()) -> (then:Exec) {is=title K2Node_CallDelegate}
  ]}
 EventGraph {nodes [
   "On Clicked (Button_Decrease)"!() -> (OutputDelegate, then, Button:Common Button Base Object Reference) {is=title K2Node_ComponentBoundEvent}
   > DecrementTime() {is=title K2Node_CallFunction}
   "On Clicked (Button_Increase)"!() -> (OutputDelegate, then, Button:Common Button Base Object Reference) {is=title K2Node_ComponentBoundEvent}
   > IncrementTime()
   Event Construct()
   > SetCompactUIStyle() {is=title K2Node_CallFunction}
   > Branch(Condition=IsUsingTouch(PlayerController=GetOwningPlayer()))
   > PlayAnimationForward(InAnimation=Get OnHoverArrows() -> (OnHoverArrows:Widget Animation Object Reference))
   Event PreConstruct()
   > "SetText (Text)"(
    self=Get Text_SettingName() -> (Text_SettingName:Common Text Object Reference)
    InText=Get Setting Header Text() -> (Setting Header Text:Text)
   )
   K2Node_Event_3 = OnInputMethodChanged!() -> (OutputDelegate, then, CurrentInput:ECommonInputType Enum) {is=title K2Node_CustomEvent}
   > Branch(Condition=IsEditable!() -> (IsEditable:bool))
   > n1 = Branch(Condition="Equal (Enum)"(A=@K2Node_Event_3.CurrentInput, B=Touch))
   n2 = Get OnHoverArrows()
   > PlayAnimationForward(InAnimation=@n2)
   n3 = Branch(execute=@n1.else, Condition=IsHovered())
   PlayAnimation(execute=@n3.else, InAnimation=@n2, StartAtTime=GetEndTime(self=Get OnHoverArrows()), PlayMode=Reverse)
   Event OnMouseEnter()
   n4 = IsUsingTouch(PlayerController=GetOwningPlayer())
   > n5 = Branch(Condition=@n4)
   n6 = GetOnHoverAnimation()
   PlayAnimationForward(execute=@n5.else, InAnimation=@n6)
   > SpawnSound2D(Sound=Get HoverSound(), VolumeMultiplier=1, PitchMultiplier=1, StartTime=0, bPersistAcrossLevelTransition=false, bAutoDestroy=true)
   n7 = GetCurrentInputType(self=Get CommonInputSubsystem())
   n8 = AND Boolean(A=OR Boolean(A="Equal (Enum)"(A=Gamepad, B=@n7), B="Equal (Enum)"(A=MouseAndKeyboard, B=@n7)), B=IsEditable())
   > Branch(Condition=@n8)
   n9 = Get OnHoverArrows()
   > PlayAnimationForward(InAnimation=@n9)
   Event OnMouseLeave!() -> (OutputDelegate, then, MouseEvent:"Pointer Event Structure (by ref)") {is=title K2Node_Event}
   > n10 = Branch(Condition=@n4)
   PlayAnimationReverse(execute=@n10.else, InAnimation=@n6)
   > Branch(Condition=@n8)
   > PlayAnimationReverse(InAnimation=@n9)
   K2Node_Event_6 = OnEditableStateChanged!() -> (OutputDelegate, then, bIsEditable:bool=~) {is=title K2Node_CustomEvent}
   > SetSettingTextEditableState!(execute, self, IsEditable:bool=@K2Node_Event_6.bIsEditable) -> (then)
   > n11 = Branch(Condition=IsUsingTouch(PlayerController=GetOwningPlayer()))
   > StopAnimation(InAnimation=Get OnHoverArrows())
   n12 = Get OnHoverArrows()
   > PlayAnimation(
    InAnimation=@n12
    StartAtTime=GetAnimationCurrentTime!(self:User Widget Object Reference, InAnimation:Widget Animation Object Reference=@n12) -> (ReturnValue:float)
    PlayMode=Select(Option 0:EUMGSequencePlayMode Enum=Reverse, Option 1:EUMGSequencePlayMode Enum, Index=@K2Node_Event_6.bIsEditable) -> (ReturnValue:EUMGSequencePlayMode Enum)
   )
   n13 = Branch(execute=@n11.else, Condition=@K2Node_Event_6.bIsEditable)
   n14 = Get OnHoverArrows()
   PlayAnimation(execute=@n13.else, InAnimation=@n14, StartAtTime=GetEndTime(self=@n14), PlayMode=Reverse)
   Event OnInitialized()
   > Assign On Input Method Changed!(execute:Exec, self:Common Input Object Reference=Get CommonInputSubsystem(), Delegate:"Delegate (by ref)"=Create Event()) -> (then:Exec) {is=title K2Node_AssignDelegate}
  ]}
 GetOnHoverAnimation {nodes [
   GetOnHoverAnimation() {is=title K2Node_FunctionEntry}
   > Return Node!(
    execute
    OnHoverAnimation:Widget Animation Object Reference=Select(
     Option 0:Widget Animation Object Reference=Get OnHoverDisabled() -> (OnHoverDisabled:Widget Animation Object Reference)
     Option 1:Widget Animation Object Reference=Get OnHover() -> (OnHover:Widget Animation Object Reference)
     Index=IsEditable()
    ) -> (ReturnValue:Widget Animation Object Reference)
   )
  ]}
 GetPrimaryGamepadFocusWidget {nodes [
   GetPrimaryGamepadFocusWidget() {is=title K2Node_FunctionEntry}
   > Return Node!(execute, ReturnValue:Widget Object Reference=Get Button_SelectEntry() -> (Button_SelectEntry:Icon Text Button Object Reference))
  ]}
 IncrementTime {nodes [
   IncrementTime() {is=title K2Node_FunctionEntry}
   n1 = Get TimeIncrementMinutes() -> (TimeIncrementMinutes=0)
   > Set Time(Time=DailyTime + DailyTime(A:Daily Time Structure=Get Time(), B:Daily Time Structure=MakeDailyTime(Hour=0, Minute=@n1)) -> (ReturnValue:Daily Time Structure) {is=title K2Node_PromotableOperator})
   > RefreshUI()
   > Call On Time Changed(NewTime=Get Time())
  ]}
 Init {nodes [
   n1 = Init!() -> (then, InitialTime:Daily Time Structure, InTimeIncrementMinutes:int) {is=title K2Node_FunctionEntry}
   > Set Time(Time=@n1)
   > Set TimeIncrementMinutes(TimeIncrementMinutes:int=@n1.InTimeIncrementMinutes) -> (Output_Get:int=0)
   > RefreshUI()
  ]}
 IsEditable {nodes [
   IsEditable() {is=title K2Node_FunctionEntry}
   > Return Node!(
    execute
    IsEditable:bool=Get bIsEnabled!(self:Widget Object Reference=Get Panel_Value() -> (Panel_Value:Horizontal Box Object Reference)) -> (bIsEnabled:bool)
   )
  ]}
 OnGamepadNavigate {nodes [
   n1 = OnGamepadNavigate!() -> (then, Navigation:EUINavigation Enum) {is=title K2Node_FunctionEntry}
   > n2 = Switch on EUINavigation!(execute, Selection:EUINavigation Enum=@n1) -> (Left:Exec, Right:Exec, Up:Exec, Down:Exec, Next:Exec, Previous:Exec, Invalid:Exec) {is=title K2Node_SwitchEnum}
   > n3 = DecrementTime()
   Return Node!(execute=[@n3.then, @n4.then], ReturnValue:Widget Object Reference)
   n4 = IncrementTime(execute=@n2.Right) {is=title K2Node_CallFunction}
  ]}
 OnTimeChanged {nodes [OnTimeChanged!() -> (then, NewTime:Daily Time Structure) {is=title K2Node_FunctionEntry}]}
 RefreshUI {nodes [
   RefreshUI() {is=title K2Node_FunctionEntry}
   > "SetText (Text)"(
    self=Get Text_Time() -> (Text_Time:Common Text Object Reference)
    InText=ToText!(DailyTime:"Daily Time Structure (by ref)"=Get Time()) -> (ReturnValue:Text)
   )
  ]}
 SetCompactUIStyle {nodes [
   SetCompactUIStyle() {is=title K2Node_FunctionEntry}
   > Branch(Condition=IsUsingCompactUI())
   > SetBrush(
    self=Get FocusOutline() -> (FocusOutline:Image Object Reference)
    InBrush=MakeBrushFromMaterial!(Material:Material Interface Object Reference=/FortniteCoreUI/Settings/Art/MI_UIKit_Settings_Outline_Bacchus, Width:int=32, Height:int=32) -> (ReturnValue:Slate Brush Structure)
   )
   > Set Icon!(
    execute
    self:Icon Text Button Object Reference=Get Button_Increase() -> (Button_Increase:Icon Text Button Object Reference)
    IconBrush_TintColor:Slate Color Structure=(SpecifiedColor=(R=1,G=1,B=1,A=1),ColorUseRule=UseColor_Specified)
    IconBrush_DrawAs:ESlateBrushDrawType Enum=Image
    IconBrush_Tiling:ESlateBrushTileType Enum=NoTile
    IconBrush_Mirroring:ESlateBrushMirrorType Enum=NoMirror
    IconBrush_ImageSize_X:double=64
    IconBrush_ImageSize_Y:double=64
    IconBrush_Margin:Margin Structure=(Left=0,Top=0,Right=0,Bottom=0)
    IconBrush_ResourceObject:Object Reference=/FortUILibrary/Textures/Icons/ToBeDeprecated/T_UI_Icon_Global_Chevron_Right
    IconBrush_OutlineSettings:Slate Brush Outline Settings Structure=(CornerRadii=(X=0,Y=0,Z=0,W=0),Color=(SpecifiedColor=(R=0,G=0,B=0,A=0),ColorUseRule=UseColor_Specified),Width=0,RoundingType=HalfHeightRadius,bUseBrushTransparency=False)
   ) -> (then) {is=title K2Node_CallFunction}
   > Set Icon(
    self=Get Button_Decrease() -> (Button_Decrease:Icon Text Button Object Reference)
    IconBrush_TintColor=(SpecifiedColor=(R=1,G=1,B=1,A=1),ColorUseRule=UseColor_Specified)
    IconBrush_DrawAs=Image
    IconBrush_Tiling=NoTile
    IconBrush_Mirroring=NoMirror
    IconBrush_ImageSize_X=64
    IconBrush_ImageSize_Y=64
    IconBrush_Margin=(Left=0,Top=0,Right=0,Bottom=0)
    IconBrush_ResourceObject=&T_UI_Icon_Global_Chevron_Right
    IconBrush_OutlineSettings=(CornerRadii=(X=0,Y=0,Z=0,W=0),Color=(SpecifiedColor=(R=0,G=0,B=0,A=0),ColorUseRule=UseColor_Specified),Width=0,RoundingType=HalfHeightRadius,bUseBrushTransparency=False)
   )
  ]}
 SetSettingTextEditableState {nodes [
   n1 = SetSettingTextEditableState!() -> (then, IsEditable:bool=~) {is=title K2Node_FunctionEntry}
   > SetColorAndOpacity(
    self=Get Text_SettingName()
    InColorAndOpacity=Select(
     Option 0:Slate Color Structure=Make Slate Color(SpecifiedColor=(R=0.2083,G=0.2083,B=0.2083,A=0.8))
     Option 1:Slate Color Structure=Make Slate Color(SpecifiedColor=(R=0.5216,G=0.5216,B=0.5216,A=0.8))
     Index=@n1
    ) -> (ReturnValue:Slate Color Structure)
   )
  ]}
]
variables [
 bIsInSelectedState {type=boolean}
 HoverSound {type=SoundBase, default=/Game/Sounds/Fort_UI_Sounds/FrontEnd/DefaultSounds/PressedButtons/MSS_UI_Button_Hover_Default}
 On Hover Arrows {type=WidgetAnimation, default=null}
 OnTimeChanged {}
 Setting Header Text {type=string, default=Add More Time}
 ShowingArrowsGamepad {type=boolean}
 Time {type=DailyTime, default {hour=0, minute=0}}
 TimeIncrementMinutes {type=integer}
]
widgets {
 info {
  inheritedWidgetCount=0
  namedSlotCount=1
  parentClass=&CommonUserWidget
  rootWidgetClass=/FortniteCoreUI/Settings/SettingEntries/SettingEntryBackground.SettingEntryBackground_C
  widgetCount=14
 }
 widgets [
  Background {widgetClassPath=/FortniteCoreUI/Settings/SettingEntries/SettingEntryBackground, bIsVariable=true, widget=Background}
  Overlay_0 {widgetClassPath=&Overlay, namedSlotHost=Background, widget=Overlay_0}
  FortResponsivePanel_85 {
   widgetClassPath=/Script/FortniteUI.FortResponsivePanel
   parent=Overlay_0
   widget=FortResponsivePanel_85
   properties {bCanStackVertically=false}
   slot {class=OverlaySlot, properties {horizontalAlignment=HAlign_Fill, verticalAlignment=VAlign_Fill, padding=<ltrb 2 2 2 2>}}
  }
  SizeBox_0 {
   widgetClassPath=&SizeBox
   parent=FortResponsivePanel_85
   widget=SizeBox_0
   properties {bOverride_MinDesiredHeight=true, minDesiredHeight=64}
   slot {class=FortResponsivePanelSlot}
  }
  Text_SettingName {
   widgetClassPath=&CommonTextBlock
   bIsVariable=true
   parent=SizeBox_0
   widget=Text_SettingName
   properties {
    scrollStyle=&TextScrollStyle-Base_C
    text="{Setting Name}"
    colorAndOpacity {colorUseRule=UseColor_Specified, specifiedColor=<rgba 0.5216 0.5216 0.5216 0.8>}
    font {
     bForceMonospaced=false
     bMaterialIsStencil=false
     fontMaterial=null
     fontObject=&HeadingNow
     letterSpacing=0
     monospacedWidth=1
     size=18
     skewAmount=0
     typefaceFontName=75 Medium
     outlineSettings {
      bApplyOutlineToDropShadows=false
      bMiteredCorners=false
      bSeparateFillAlpha=false
      outlineMaterial=null
      outlineSize=0
      outlineColor=<rgba 0 0 0 1>
     }
    }
    shadowOffset=<0 0>
   }
   slot {class=SizeBoxSlot, properties {horizontalAlignment=HAlign_Left, verticalAlignment=VAlign_Center, padding=<ltrb 24 0 10 0>}}
  }
  Panel_Value {
   widgetClassPath=&HorizontalBox
   bIsVariable=true
   parent=FortResponsivePanel_85
   widget=Panel_Value
   slot {class=FortResponsivePanelSlot}
  }
  Several_Options {
   widgetClassPath=&HorizontalBox
   bIsVariable=true
   parent=Panel_Value
   widget=Several_Options
   slot {class=HorizontalBoxSlot, properties {size {sizeRule=Fill, value=1}}}
  }
  Button_Decrease {
   widgetClassPath=&IconTextButton
   bIsVariable=true
   parent=Several_Options
   widget=Button_Decrease
   properties {
    bHideInputAction=true
    bInteractableWhenSelected=true
    bToggleable=true
    buttonText=""
    onButtonBaseClicked=(&self:WidgetTree.Button_Decrease.HandleOnButtonClicked)
    onButtonBaseDoubleClicked=(&self:WidgetTree.Button_Decrease.HandleOnButtonDoubleClicked)
    onButtonBaseHovered=(&self:WidgetTree.Button_Decrease.HandleOnButtonHovered)
    onButtonBaseUnhovered=(&self:WidgetTree.Button_Decrease.HandleOnButtonUnhovered)
    onSelectedChangedBase=(&self:WidgetTree.Button_Decrease.HandleOnSelectedChanged)
    style=/FortUILibrary/Foundation/Buttons/ButtonStyles/ButtonStyleClearDismissTooltip.ButtonStyleClearDismissTooltip_C
    touchMethod=PreciseTap
    colorAndOpacity=<rgba 1 1 1 0>
    iconBrush {
     bIsDynamicallyLoaded=false
     drawAs=Image
     imageType=NoImage
     mirroring=NoMirror
     resourceName=None
     resourceObject=&T_UI_Icon_Global_Chevron_Right
     tiling=NoTile
     imageSize=<32 32>
     margin=<ltrb 0 0 0 0>
     outlineSettings {
      bUseBrushTransparency=false
      roundingType=HalfHeightRadius
      width=0
      color {colorUseRule=UseColor_Specified, specifiedColor=<rgba 0 0 0 0>}
      cornerRadii=<0 0 0 1>
     }
     tintColor {colorUseRule=UseColor_Specified, specifiedColor=<rgba 1 1 1 1>}
     uVRegion {bIsValid=false, max=<0 0>, min=<0 0>}
    }
    renderTransform {angle=0, scale=<-1 1>, shear=<0 0>, translation=<0 0>}
   }
   slot {class=HorizontalBoxSlot}
  }
  Border_0 {
   widgetClassPath=&Border
   parent=Several_Options
   widget=Border_0
   properties {horizontalAlignment=HAlign_Center, verticalAlignment=VAlign_Center, brushColor=<rgba 1 1 1 0>, padding=<ltrb 0 0 0 0>}
   slot {class=HorizontalBoxSlot, properties {size {sizeRule=Fill, value=1}}}
  }
  Text_Time {
   widgetClassPath=&CommonTextBlock
   bIsVariable=true
   parent=Border_0
   widget=Text_Time
   properties {
    style=&TextStyle-Base_C
    text=11:00 AM
    font {
     bForceMonospaced=false
     bMaterialIsStencil=false
     fontMaterial=null
     fontObject=&NotoSans
     letterSpacing=0
     monospacedWidth=1
     size=24
     skewAmount=0
     typefaceFontName=NotoSans-Regular
     outlineSettings {
      bApplyOutlineToDropShadows=false
      bMiteredCorners=false
      bSeparateFillAlpha=false
      outlineMaterial=null
      outlineSize=0
      outlineColor=<rgba 0 0 0 1>
     }
    }
    shadowOffset=<0 0>
   }
   slot {class=BorderSlot, properties {horizontalAlignment=HAlign_Center, verticalAlignment=VAlign_Center, padding=<ltrb 0 0 0 0>}}
  }
  Button_Increase {
   widgetClassPath=&IconTextButton
   bIsVariable=true
   parent=Several_Options
   widget=Button_Increase
   properties {
    bHideInputAction=true
    bInteractableWhenSelected=true
    bToggleable=true
    buttonText=""
    onButtonBaseClicked=(&self:WidgetTree.Button_Increase.HandleOnButtonClicked)
    onButtonBaseDoubleClicked=(&self:WidgetTree.Button_Increase.HandleOnButtonDoubleClicked)
    onButtonBaseHovered=(&self:WidgetTree.Button_Increase.HandleOnButtonHovered)
    onButtonBaseUnhovered=(&self:WidgetTree.Button_Increase.HandleOnButtonUnhovered)
    onSelectedChangedBase=(&self:WidgetTree.Button_Increase.HandleOnSelectedChanged)
    style=&ButtonStyleClearDismissTooltip_C
    touchMethod=PreciseTap
    colorAndOpacity=<rgba 1 1 1 0>
    iconBrush {
     bIsDynamicallyLoaded=false
     drawAs=Image
     imageType=NoImage
     mirroring=NoMirror
     resourceName=None
     resourceObject=&T_UI_Icon_Global_Chevron_Right
     tiling=NoTile
     imageSize=<32 32>
     margin=<ltrb 0 0 0 0>
     outlineSettings {
      bUseBrushTransparency=false
      roundingType=HalfHeightRadius
      width=0
      color {colorUseRule=UseColor_Specified, specifiedColor=<rgba 0 0 0 0>}
      cornerRadii=<0 0 0 1>
     }
     tintColor {colorUseRule=UseColor_Specified, specifiedColor=<rgba 1 1 1 1>}
     uVRegion {bIsValid=false, max=<0 0>, min=<0 0>}
    }
   }
   slot {class=HorizontalBoxSlot}
  }
  SizeBox_1 {
   widgetClassPath=&SizeBox
   parent=Overlay_0
   widget=SizeBox_1
   properties {bOverride_WidthOverride=true, widthOverride=50}
   slot {class=OverlaySlot, properties {horizontalAlignment=HAlign_Right, verticalAlignment=VAlign_Fill}}
  }
  Button_SelectEntry {
   widgetClassPath=&IconTextButton
   bIsVariable=true
   parent=SizeBox_1
   widget=Button_SelectEntry
   properties {
    buttonText=""
    contentAlignment=HAlign_Center
    onButtonBaseClicked=()
    onButtonBaseDoubleClicked=()
    onButtonBaseHovered=()
    onButtonBaseUnhovered=()
    onSelectedChangedBase=()
    style=&ButtonStyleClearDismissTooltip_C
    touchMethod=PreciseTap
    useLightInput=true
    visibility=Collapsed
    font {
     bForceMonospaced=false
     bMaterialIsStencil=false
     fontMaterial=null
     fontObject=&HeadingNow
     letterSpacing=0
     monospacedWidth=1
     size=14.25
     skewAmount=0
     typefaceFontName=75 Medium
     outlineSettings {
      bApplyOutlineToDropShadows=false
      bMiteredCorners=false
      bSeparateFillAlpha=false
      outlineMaterial=null
      outlineSize=0
      outlineColor=<rgba 0 0 0 1>
     }
    }
   }
   slot {class=SizeBoxSlot}
  }
  FocusOutline {
   widgetClassPath=&Image
   bIsVariable=true
   parent=Overlay_0
   widget=FocusOutline
   properties {
    pixelSnapping=Disabled
    renderOpacity=0
    visibility=SelfHitTestInvisible
    brush {
     bIsDynamicallyLoaded=false
     drawAs=Image
     imageType=NoImage
     mirroring=NoMirror
     resourceName=None
     resourceObject=/FortniteCoreUI/Settings/Art/MI_UIKit_Settings_Outline
     tiling=NoTile
     imageSize=<32 32>
     margin=<ltrb 0 0 0 0>
     outlineSettings {
      bUseBrushTransparency=false
      roundingType=HalfHeightRadius
      width=0
      color {colorUseRule=UseColor_Specified, specifiedColor=<rgba 0 0 0 0>}
      cornerRadii=<0 0 0 0>
     }
     tintColor {colorUseRule=UseColor_Specified, specifiedColor=<rgba 1 1 1 1>}
     uVRegion {bIsValid=false, max=<0 0>, min=<0 0>}
    }
   }
   slot {class=OverlaySlot, properties {horizontalAlignment=HAlign_Fill, verticalAlignment=VAlign_Fill, padding=<ltrb -16 -16 -16 -16>}}
  }
 ]
}
