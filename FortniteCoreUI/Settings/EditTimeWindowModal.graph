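# Modal dialog in Fortnite settings for adding, editing or removing a daily time window, built on the native EditTimeWindowModal class.
# OnInitializeTimeWindow sets title, Remove button and info text for new versus existing windows, then seeds the start and end selectors.
# Selector changes update StartTime and EndTime, then RefreshUI revalidates, toggling Accept and showing the error reason.
# Accept revalidates before updating; selectors get custom gamepad navigation, and compact UI restyles the dialog.
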WidgetBlueprint /FortniteCoreUI/Settings/EditTimeWindowModal
graphs [
 BP_GetDesiredFocusTarget {nodes [
   Get Desired Focus Target()
   > Return Node!(execute, ReturnValue:Widget Object Reference=Get StartTimeSelector() -> (StartTimeSelector:WBP Date Time Selector Object Reference))
  ]}
 EventGraph {nodes [
   "On Clicked (AcceptButton)"!() -> (OutputDelegate, then, Button:Common Button Base Object Reference) {is=title K2Node_ComponentBoundEvent}
   n1 = Make Daily Time Window!(
    StartTime:Daily Time Structure=Get StartTime() -> (StartTime:Daily Time Structure)
    EndTime:Daily Time Structure=Get EndTime() -> (EndTime:Daily Time Structure)
   ) -> (DailyTimeWindow:Daily Time Window Structure) {is=title K2Node_MakeStruct}
   > n2 = IsTimeWindowValid!(execute, self:Edit Time Window Modal Object Reference, TimeWindow:"Daily Time Window Structure (by ref)"=@n1) -> (then, OutIsValid:bool, OutInvalidReason:Text)
   > Branch(Condition=@n2)
   > UpdateTimeWindow!(execute, self:Edit Time Window Modal Object Reference, TimeWindow:"Daily Time Window Structure (by ref)"=@n1) -> (then)
   > CloseModal(self:Edit Time Window Modal Object Reference)
   "On Clicked (CloseButton)"!() -> (OutputDelegate, then, Button:Common Button Base Object Reference) {is=title K2Node_ComponentBoundEvent}
   > CloseModal()
   "On Clicked (RemoveButton)"!() -> (OutputDelegate, then, Button:Common Button Base Object Reference) {is=title K2Node_ComponentBoundEvent}
   > RemoveTimeWindow(self:Edit Time Window Modal Object Reference)
   > CloseModal()
   n3 = "On Time Changed (StartTimeSelector)"!() -> (OutputDelegate, then, NewTime:Daily Time Structure) {is=title K2Node_ComponentBoundEvent}
   > n4 = Set StartTime(StartTime:Daily Time Structure=@n3.NewTime) -> (Output_Get:Daily Time Structure)
   RefreshUI(execute=[@n4.then, @n6.then])
   n5 = "On Time Changed (EndTimeSelector)"!() -> (OutputDelegate, then, NewTime:Daily Time Structure) {is=title K2Node_ComponentBoundEvent}
   > n6 = Set EndTime(EndTime:Daily Time Structure=@n5.NewTime) -> (Output_Get:Daily Time Structure)
   OnInputMethodChanged!() -> (OutputDelegate, then, NewInputType:ECommonInputType Enum) {is=title K2Node_CustomEvent}
   > RequestRefreshFocus(self:Common Activatable Widget Object Reference)
   RefreshUI() {is=title K2Node_CustomEvent}
   > n7 = IsTimeWindowValid(TimeWindow=Make Daily Time Window(StartTime=Get StartTime(), EndTime=Get EndTime()))
   > SetIsEnabled(self=Get AcceptButton() -> (AcceptButton:WBP UIKit Button Regular Object Reference), bInIsEnabled=@n7)
   n8 = Get ErrorText() -> (ErrorText:Common Text Object Reference)
   > SetVisibility(self=@n8, InVisibility=Select(Option 0=HitTestInvisible, Option 1=Hidden, Index=@n7))
   > "SetText (Text)"(self=@n8, InText=@n7.OutInvalidReason)
   Event PreConstruct()
   > SetCompactUIStyle() {is=title K2Node_CallFunction}
   n9 = Event OnInitializeTimeWindow!() -> (OutputDelegate, then, TimeWindow:"Daily Time Window Structure (by ref)", bIsNewWindow:bool) {is=title K2Node_Event}
   > Set Visibility(
    Visibility=Select(Option 0=Visible, Option 1=Collapsed, Index=@n9.bIsNewWindow)
    self=Get RemoveButton() -> (RemoveButton:WBP UIKit Button Regular Object Reference)
   )
   > Set Visibility(
    Visibility=Select(Option 0=Collapsed, Option 1=SelfHitTestInvisible, Index=@n9.bIsNewWindow)
    self=Get InfoText() -> (InfoText:Common Text Object Reference)
   )
   > "SetText (Text)"(
    self=Get Title() -> (Title:Common Text Object Reference)
    InText=Select(Option 0:Text="LOCTEXT(\"EDIT WINDOW\")", Option 1:Text="LOCTEXT(\"ADD ANOTHER WINDOW\")", Index=@n9.bIsNewWindow) -> (ReturnValue:Text)
   )
   n10 = Break Daily Time Window!(DailyTimeWindow:"Daily Time Window Structure (by ref)"=@n9.TimeWindow) -> (StartTime:Daily Time Structure, EndTime:Daily Time Structure) {is=title K2Node_BreakStruct}
   > Set StartTime(StartTime=@n10)
   > Set EndTime(EndTime=@n10.EndTime)
   > RefreshUI() {is=title K2Node_CallFunction}
   n11 = GetTimeIncrementMinutes!(self:Edit Time Window Modal Object Reference) -> (ReturnValue:int=0)
   > Init!(
    execute
    self:WBP Date Time Selector Object Reference=Get StartTimeSelector()
    InitialTime:Daily Time Structure=Get StartTime()
    InTimeIncrementMinutes:int=@n11
   ) -> (then)
   > Init(
    self=Get EndTimeSelector() -> (EndTimeSelector:WBP Date Time Selector Object Reference)
    InitialTime=Get EndTime()
    InTimeIncrementMinutes=@n11
   )
   Event On Activated()
   > SetVisibility(
    self=Get ButtonsContainer!(self:WBP UIKit Dialog Internal Object Reference=Get Dialog_Internal() -> (Dialog_Internal:WBP UIKit Dialog Internal Object Reference)) -> (ButtonsContainer:Horizontal Box Object Reference)
    InVisibility=Collapsed
   )
   Event Construct()
   n12 = Create Event()
   n13 = Get CommonInputSubsystem()
   > Bind Event to On Input Method Changed(self=@n13, Delegate=@n12)
   > OnInputMethodChanged!(execute, self, NewInputType:ECommonInputType Enum=GetCurrentInputType(self=Get CommonInputSubsystem())) -> (then) {is=title K2Node_CallFunction}
   Event Destruct()
   > Unbind Event from On Input Method Changed(self=@n13, Delegate=@n12)
  ]}
 OnEndTimeSelectorNavigated {nodes [
   n1 = OnEndTimeSelectorNavigated!() -> (then, Navigation:EUINavigation Enum) {is=title K2Node_FunctionEntry}
   > n2 = OnGamepadNavigate!(execute, self:WBP Date Time Selector Object Reference=Get EndTimeSelector(), Navigation:EUINavigation Enum=@n1) -> (then, ReturnValue:Widget Object Reference)
   > Return Node(ReturnValue=@n2)
  ]}
 OnStartTimeSelectorNavigated {nodes [
   n1 = OnStartTimeSelectorNavigated!() -> (then, Navigation:EUINavigation Enum) {is=title K2Node_FunctionEntry}
   > n2 = OnGamepadNavigate(self=Get StartTimeSelector(), Navigation=@n1)
   > Return Node(ReturnValue=@n2)
  ]}
 SetCompactUIStyle {nodes [
   SetCompactUIStyle() {is=title K2Node_FunctionEntry}
   > Branch(Condition=IsUsingCompactUI())
   > SetSizeBox!(
    execute
    self:WBP UIKit Dialog Internal Object Reference=Get Dialog_Internal()
    Data_WidthOverride_107_#1_Enable_3_#2:bool=true
    Data_WidthOverride_107_#1_Value_4_#3:double=1024
    Data_HeightOverride_108_#4:S UI Bool Float Structure=(Enable_3_#2=False,Value_4_#3=0)
    Data_MinDesiredWidth_109_#5:S UI Bool Float Structure=(Enable_3_#2=False,Value_4_#3=0)
    Data_MinDesiredHeight_110_#6:S UI Bool Float Structure=(Enable_3_#2=False,Value_4_#3=0)
    Data_MaxDesiredWidth_111_#7:S UI Bool Float Structure=(Enable_3_#2=False,Value_4_#3=0)
    Data_MaxDesiredHeight_112_#8:S UI Bool Float Structure=(Enable_3_#2=False,Value_4_#3=0)
    Data_MinAspectRatio_113_#9:S UI Bool Float Structure=(Enable_3_#2=False,Value_4_#3=0)
    Data_MaxAspectRatio_114_#10:S UI Bool Float Structure=(Enable_3_#2=False,Value_4_#3=0)
   ) -> (then)
   n1 = Get InfoText()
   > SetFontSize!(execute, self:Text Object Reference=[@n1, Get ErrorText()], DisplayFontSize:float=24) -> (then)
   > SetAutoWrapText!(execute, self:Text Object Reference=@n1, InAutoTextWrap:bool=true) -> (then)
  ]}
]
variables [
 EndTime {type=DailyTime, default {hour=0, minute=0}}
 StartTime {type=DailyTime, default {hour=0, minute=0}}
]
widgets {
 info {
  inheritedWidgetCount=0
  namedSlotCount=1
  parentClass=/Script/FortniteUI.EditTimeWindowModal
  rootWidgetClass=/FortUILibrary/Systems/Dialog/WBP_UIKit_Dialog_Internal.WBP_UIKit_Dialog_Internal_C
  widgetCount=12
 }
 widgets [
  Dialog_Internal {
   widgetClassPath=/FortUILibrary/Systems/Dialog/WBP_UIKit_Dialog_Internal
   bIsVariable=true
   widget=Dialog_Internal
   properties {
    useText=None
    useTitle=false
    dialogSIze {
     height Override {enable=false, value=800}
     max Aspect Ratio {enable=false, value=0}
     max Desired Height {enable=true, value=800}
     max Desired Width {enable=false, value=0}
     min Aspect Ratio {enable=false, value=0}
     min Desired Height {enable=false, value=0}
     min Desired Width {enable=false, value=0}
     width Override {enable=true, value=864}
    }
   }
  }
  Overlay_0 {widgetClassPath=&Overlay, namedSlotHost=Dialog_Internal, widget=Overlay_0}
  Content {
   widgetClassPath=&VerticalBox
   parent=Overlay_0
   widget=Content
   slot {class=OverlaySlot, properties {horizontalAlignment=HAlign_Fill, verticalAlignment=VAlign_Fill, padding=<ltrb 80 64 80 48>}}
  }
  Title {
   widgetClassPath=&CommonTextBlock
   bIsVariable=true
   parent=Content
   widget=Title
   properties {
    applyLineHeightToBottomLine=false
    autoWrapText=true
    text=ADD ANOTHER WINDOW
    textTransformPolicy=ToUpper
    font {
     bForceMonospaced=false
     bMaterialIsStencil=false
     fontMaterial=null
     fontObject=&HeadingNow
     letterSpacing=-10
     monospacedWidth=1
     size=25.5
     skewAmount=0
     typefaceFontName=86 Bold
     outlineSettings {
      bApplyOutlineToDropShadows=false
      bMiteredCorners=false
      bSeparateFillAlpha=false
      outlineMaterial=null
      outlineSize=0
      outlineColor=<rgba 0 0 0 1>
     }
    }
    shadowOffset=<0 0>
    strikeBrush {
     bIsDynamicallyLoaded=false
     drawAs=NoDrawType
     imageType=NoImage
     mirroring=NoMirror
     resourceName=None
     resourceObject=null
     tiling=NoTile
     imageSize=<32 32>
     margin=<ltrb 0 0 0 0>
     outlineSettings {
      bUseBrushTransparency=false
      roundingType=HalfHeightRadius
      width=0
      color {colorUseRule=UseColor_Specified, specifiedColor=<rgba 0 0 0 0>}
      cornerRadii=<0 0 0 0>
     }
     tintColor {colorUseRule=UseColor_Specified, specifiedColor=<rgba 1 1 1 1>}
     uVRegion {bIsValid=false, max=<0 0>, min=<0 0>}
    }
   }
   slot {class=VerticalBoxSlot, properties {verticalAlignment=VAlign_Top, padding=<ltrb 0 0 0 32>}}
  }
  StartTimeSelector {
   widgetClassPath=&/WBP_DateTimeSelector
   bIsVariable=true
   parent=Content
   widget=StartTimeSelector
   properties {
    navigation=&self:WidgetTree.StartTimeSelector.WidgetNavigation_0
    setting Header Text=Start Time
    visibility=Visible
   }
   slot {class=VerticalBoxSlot}
  }
  EndTimeSelector {
   widgetClassPath=&WBP_DateTimeSelector
   bIsVariable=true
   parent=Content
   widget=EndTimeSelector
   properties {
    navigation=&self:WidgetTree.EndTimeSelector.WidgetNavigation_0
    setting Header Text=End Time
    visibility=Visible
   }
   slot {class=VerticalBoxSlot}
  }
  InfoText {
   widgetClassPath=&CommonTextBlock
   bIsVariable=true
   parent=Content
   widget=InfoText
   properties {
    applyLineHeightToBottomLine=false
    lineHeightPercentage=1.05
    text=You can edit or remove this time window later.
    font {
     bForceMonospaced=false
     bMaterialIsStencil=false
     fontMaterial=null
     fontObject=&HeadingNow
     letterSpacing=10
     monospacedWidth=1
     size=18
     skewAmount=0
     typefaceFontName=74 Regular
     outlineSettings {
      bApplyOutlineToDropShadows=false
      bMiteredCorners=false
      bSeparateFillAlpha=false
      outlineMaterial=null
      outlineSize=0
      outlineColor=<rgba 0 0 0 1>
     }
    }
    shadowOffset=<0 0>
    strikeBrush {
     bIsDynamicallyLoaded=false
     drawAs=NoDrawType
     imageType=NoImage
     mirroring=NoMirror
     resourceName=None
     resourceObject=null
     tiling=NoTile
     imageSize=<32 32>
     margin=<ltrb 0 0 0 0>
     outlineSettings {
      bUseBrushTransparency=false
      roundingType=HalfHeightRadius
      width=0
      color {colorUseRule=UseColor_Specified, specifiedColor=<rgba 0 0 0 0>}
      cornerRadii=<0 0 0 0>
     }
     tintColor {colorUseRule=UseColor_Specified, specifiedColor=<rgba 1 1 1 1>}
     uVRegion {bIsValid=false, max=<0 0>, min=<0 0>}
    }
   }
   slot {class=VerticalBoxSlot, properties {horizontalAlignment=HAlign_Left, verticalAlignment=VAlign_Center, padding=<ltrb 0 8 0 0>}}
  }
  ErrorText {
   widgetClassPath=&CommonTextBlock
   bIsVariable=true
   parent=Content
   widget=ErrorText
   properties {
    applyLineHeightToBottomLine=false
    autoWrapText=true
    lineHeightPercentage=1.05
    text="Oops! Something went wrong. Please try again."
    colorAndOpacity {colorUseRule=UseColor_Specified, specifiedColor=<rgba 0.5776 0.0451 0.0451 1>}
    font {
     bForceMonospaced=false
     bMaterialIsStencil=false
     fontMaterial=null
     fontObject=&HeadingNow
     letterSpacing=10
     monospacedWidth=1
     size=18
     skewAmount=0
     typefaceFontName=74 Regular
     outlineSettings {
      bApplyOutlineToDropShadows=false
      bMiteredCorners=false
      bSeparateFillAlpha=false
      outlineMaterial=null
      outlineSize=0
      outlineColor=<rgba 0 0 0 1>
     }
    }
    shadowOffset=<0 0>
    strikeBrush {
     bIsDynamicallyLoaded=false
     drawAs=NoDrawType
     imageType=NoImage
     mirroring=NoMirror
     resourceName=None
     resourceObject=null
     tiling=NoTile
     imageSize=<32 32>
     margin=<ltrb 0 0 0 0>
     outlineSettings {
      bUseBrushTransparency=false
      roundingType=HalfHeightRadius
      width=0
      color {colorUseRule=UseColor_Specified, specifiedColor=<rgba 0 0 0 0>}
      cornerRadii=<0 0 0 0>
     }
     tintColor {colorUseRule=UseColor_Specified, specifiedColor=<rgba 1 1 1 1>}
     uVRegion {bIsValid=false, max=<0 0>, min=<0 0>}
    }
   }
   slot {class=VerticalBoxSlot, properties {horizontalAlignment=HAlign_Left, verticalAlignment=VAlign_Center, padding=<ltrb 0 16 0 0>}}
  }
  HorizontalBox_0 {
   widgetClassPath=&HorizontalBox
   parent=Content
   widget=HorizontalBox_0
   slot {class=VerticalBoxSlot, properties {verticalAlignment=VAlign_Bottom, padding=<ltrb 0 64 0 0>}}
  }
  AcceptButton {
   widgetClassPath=&WBP_UIKit_Button_Regular
   bIsVariable=true
   parent=HorizontalBox_0
   widget=AcceptButton
   properties {
    minHeight=56
    navigation=&self:WidgetTree.AcceptButton.WidgetNavigation_0
    text=Accept
    visibility=Visible
    triggeringInputAction {dataTable=&InputActionDataTable, rowName=Accept}
   }
   slot {class=HorizontalBoxSlot, properties {verticalAlignment=VAlign_Center, padding=<ltrb 0 0 24 0>, size {sizeRule=Fill, value=1}}}
  }
  RemoveButton {
   widgetClassPath=&WBP_UIKit_Button_Regular
   bIsVariable=true
   parent=HorizontalBox_0
   widget=RemoveButton
   properties {
    minHeight=56
    navigation=&self:WidgetTree.RemoveButton.WidgetNavigation_0
    text=Remove
    visibility=Visible
    triggeringInputAction {dataTable=&InputActionDataTable, rowName=GenericFaceButtonTop}
   }
   slot {class=HorizontalBoxSlot, properties {verticalAlignment=VAlign_Center, padding=<ltrb 0 0 24 0>, size {sizeRule=Fill, value=1}}}
  }
  CloseButton {
   widgetClassPath=&WBP_UIKit_Button_Regular
   bIsVariable=true
   parent=HorizontalBox_0
   widget=CloseButton
   properties {
    buttonType=Quiet
    minHeight=56
    navigation=&self:WidgetTree.CloseButton.WidgetNavigation_0
    text=Close
    visibility=Visible
    triggeringInputAction {dataTable=&InputActionDataTable, rowName=Close}
   }
   slot {class=HorizontalBoxSlot, properties {size {sizeRule=Fill, value=1}}}
  }
 ]
}
