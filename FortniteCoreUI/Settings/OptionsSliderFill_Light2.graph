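# Translucent UI material for an options slider fill, drawing a thin horizontal track, a center tick and a position marker.
# Masks come from rounded, inverted UV distances: track at mid-height, tick at U 0.5, marker at the fillPercent parameter.
# The masks are summed into a combined shape, while a lerp blends Color BG to Color Marker at the marker.
# The tick lerp uses Color BG on both sides, and ColorMiddle is unconnected.
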Material /FortniteCoreUI/Settings/OptionsSliderFill_Light2
graph {
 properties {
  bUsedWithStaticMesh=false
  blendMode=BLEND_Translucent
  materialDomain=MD_UI
  pixelDepthOffsetMode=PDOM_Legacy
  usedShadingModels=Default Lit
 }
 nodes [
  n1 = ComponentMask(None=TextureCoordinate()) {r=true}
  n2 = Multiply(A=Abs(None=Add(A=@n1) {constB=-0.5}), B=Constant() {r=200}) {constB=500}
  n3 = Saturate(None=Round(None=OneMinus(None=@n2)))
  n4 = TextureCoordinate()
  n5 = Abs(None=Add(A=ComponentMask(None=@n4) {g=true}) {constB=-0.5})
  n6 = Round(None=OneMinus(None=Multiply(A=@n5) {constB=10}))
  n7 = ComponentMask(None=TextureCoordinate()) {r=true}
  n8 = Multiply(A=ScalarParameter() {defaultValue=0.3, group=xp, parameterName=fillPercent, sliderMax=1}) {constB=-1}
  n9 = Multiply(A=Abs(None=Add(A=@n7, B=@n8) {constB=-0.3}), B=Constant() {r=250}) {constB=500}
  n10 = Saturate(None=Round(None=OneMinus(None=@n9)))
  Add(A=@n3, B=Add(A=Saturate(None=@n6), B=@n10))
  n11 = VectorParameter() {parameterName=Color BG, defaultValue=<rgba 0.0666 0.147 0.5395 1>}
  LinearInterpolate(
   A=LinearInterpolate(A=@n11.RGB, B=@n11.RGB, Alpha=@n3) {constA=0.03, constB=0.1}
   B=VectorParameter() {parameterName=Color Marker, defaultValue=<rgba 1 1 1 1>}.RGB
   Alpha=@n10
  )
  VectorParameter() {parameterName=ColorMiddle, defaultValue=<rgba 0.0185 0.0037 0.0865 1>}
 ]
}
