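# Widget Blueprint for a modal progress dialog, based on the native FortProgressWidget class.
# A centred fixed-size bordered panel holds a header with icon and title above a body with a throbber-material spinner and wrapped description text.
# The icon, title and description are inherited widgets placed in this layout.
# The event graph is empty, so the Blueprint supplies only layout and styling.
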WidgetBlueprint /Game/UI/Foundation/Progress/ProgressWidget
graphs [EventGraph {links [], nodes []}]
variables []
widgets {
 info {inheritedWidgetCount=3, parentClass=/Script/FortniteUI.FortProgressWidget, rootWidgetClass=&Overlay, widgetCount=13}
 widgets [
  Overlay_0 {widgetClassPath=&Overlay, widget=Overlay_0}
  SizeBox_0 {
   widgetClassPath=&SizeBox
   parent=Overlay_0
   widget=SizeBox_0
   properties {bOverride_HeightOverride=true, bOverride_WidthOverride=true, heightOverride=600, widthOverride=1000}
   slot {class=OverlaySlot, properties {horizontalAlignment=HAlign_Center, verticalAlignment=VAlign_Center}}
  }
  CommonBorder_0 {
   widgetClassPath=&CommonBorder
   parent=SizeBox_0
   widget=CommonBorder_0
   properties {style=/FortUILibrary/Styles/Border/Border-MainModal.Border-MainModal_C, padding=<ltrb 0 0 0 0>}
   slot {class=SizeBoxSlot}
  }
  VerticalBox_0 {
   widgetClassPath=&VerticalBox
   parent=CommonBorder_0
   widget=VerticalBox_0
   slot {class=BorderSlot, properties {padding=<ltrb 0 0 0 0>}}
  }
  CommonBorder_Modal {
   widgetClassPath=&CommonBorder
   parent=VerticalBox_0
   widget=CommonBorder_Modal
   properties {style=/FortUILibrary/Styles/Border/Border-ModalHeader.Border-ModalHeader_C}
   slot {class=VerticalBoxSlot}
  }
  HorizontalBox_1 {widgetClassPath=&HorizontalBox, parent=CommonBorder_Modal, widget=HorizontalBox_1, slot {class=BorderSlot}}
  Spacer_0 {widgetClassPath=&Spacer, parent=HorizontalBox_1, widget=Spacer_0, properties {size=<100 100>}, slot {class=HorizontalBoxSlot}}
  Image_MainIcon {
   widgetClassPath=&Image
   bInherited=true
   bIsVariable=true
   parent=HorizontalBox_1
   widget=Image_MainIcon
   properties {brush {
     bIsDynamicallyLoaded=false
     drawAs=Image
     imageType=NoImage
     mirroring=NoMirror
     resourceName=None
     resourceObject=null
     tiling=NoTile
     imageSize=<64 64>
     margin=<ltrb 0 0 0 0>
     outlineSettings {
      bUseBrushTransparency=false
      roundingType=HalfHeightRadius
      width=0
      color {colorUseRule=UseColor_Specified, specifiedColor=<rgba 0 0 0 0>}
      cornerRadii=<0 0 0 0>
     }
     tintColor {colorUseRule=UseColor_Specified, specifiedColor=<rgba 1 1 1 1>}
     uVRegion {bIsValid=false, max=<0 0>, min=<0 0>}
    }}
   slot {class=HorizontalBoxSlot, properties {horizontalAlignment=HAlign_Left, verticalAlignment=VAlign_Center}}
  }
  Text_Title {
   widgetClassPath=&CommonTextBlock
   bInherited=true
   bIsVariable=true
   parent=HorizontalBox_1
   widget=Text_Title
   properties {
    scrollStyle=&TextScrollStyle-Base_C
    style=/FortUILibrary/Styles/Typography/Foundation/TextStyle-Header-L.TextStyle-Header-L_C
    text=Title
    font {
     bForceMonospaced=false
     bMaterialIsStencil=false
     fontMaterial=null
     fontObject=&Burbank
     letterSpacing=0
     monospacedWidth=1
     size=48
     skewAmount=0
     typefaceFontName=Burbank-Black
     outlineSettings {
      bApplyOutlineToDropShadows=false
      bMiteredCorners=false
      bSeparateFillAlpha=false
      outlineMaterial=null
      outlineSize=0
      outlineColor=<rgba 0 0 0 1>
     }
    }
    shadowOffset=<0 0>
   }
   slot {
    class=HorizontalBoxSlot
    properties {horizontalAlignment=HAlign_Left, verticalAlignment=VAlign_Center, padding=<ltrb 8 0 0 0>, size {sizeRule=Fill, value=1}}
   }
  }
  ProgressSizeBox {
   widgetClassPath=&SizeBox
   bIsVariable=true
   parent=VerticalBox_0
   widget=ProgressSizeBox
   slot {class=VerticalBoxSlot, properties {padding=<ltrb 20 20 20 20>, size {sizeRule=Fill, value=1}}}
  }
  ProgressVBox {
   widgetClassPath=&VerticalBox
   bIsVariable=true
   parent=ProgressSizeBox
   widget=ProgressVBox
   slot {class=SizeBoxSlot, properties {verticalAlignment=VAlign_Center}}
  }
  ProgressSpinner {
   widgetClassPath=&Image
   bIsVariable=true
   parent=ProgressVBox
   widget=ProgressSpinner
   properties {brush {
     bIsDynamicallyLoaded=false
     drawAs=Image
     imageType=NoImage
     mirroring=NoMirror
     resourceName=None
     resourceObject=/FortUILibrary/UIKit/Throbber/MI_UIKit_Color_Throbber
     tiling=NoTile
     imageSize=<64 64>
     margin=<ltrb 0 0 0 0>
     outlineSettings {
      bUseBrushTransparency=false
      roundingType=HalfHeightRadius
      width=0
      color {colorUseRule=UseColor_Specified, specifiedColor=<rgba 0 0 0 0>}
      cornerRadii=<0 0 0 0>
     }
     tintColor {colorUseRule=UseColor_Specified, specifiedColor=<rgba 1 1 1 1>}
     uVRegion {bIsValid=false, max=<0 0>, min=<0 0>}
    }}
   slot {class=VerticalBoxSlot, properties {horizontalAlignment=HAlign_Center, verticalAlignment=VAlign_Center}}
  }
  Text_description {
   widgetClassPath=&CommonTextBlock
   bInherited=true
   bIsVariable=true
   parent=ProgressVBox
   widget=Text_description
   properties {
    autoWrapText=true
    justification=Center
    style=/FortUILibrary/Styles/Typography/Foundation/TextStyle-Base-M-B_Blue.TextStyle-Base-M-B_Blue_C
    text=Description Progress...
    colorAndOpacity {colorUseRule=UseColor_Specified, specifiedColor=<rgba 0.008 0.011 0.0262 1>}
    font {
     bForceMonospaced=false
     bMaterialIsStencil=false
     fontMaterial=null
     fontObject=&NotoSans
     letterSpacing=0
     monospacedWidth=1
     size=24
     skewAmount=0
     typefaceFontName=NotoSans-Bold
     outlineSettings {
      bApplyOutlineToDropShadows=false
      bMiteredCorners=false
      bSeparateFillAlpha=false
      outlineMaterial=null
      outlineSize=0
      outlineColor=<rgba 0 0 0 1>
     }
    }
    shadowOffset=<0 0>
   }
   slot {class=VerticalBoxSlot}
  }
 ]
}
