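# Fullscreen progress modal for settings, built on FortProgressModal, showing a centered throbber in a Lightbox over a background blur.
# The inherited progress widget and its container are collapsed, leaving only the spinner visible by default.
# Construct binds Lightbox intro/outro events and optionally auto-initializes; intro and outro either animate through the Lightbox or end immediately when disabled.
# Analog input is consumed when bConsumeAnalogInput is set.
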WidgetBlueprint /FortniteCoreUI/Settings/SettingProgressModal
graphs [
 EventGraph {nodes [
   HandleIntroEnded() {is=title K2Node_CustomEvent}
   > EndIntro(self:Common Activatable Panel Legacy Object Reference)
   HandleOutroEnded() {is=title K2Node_CustomEvent}
   > EndOutro(self:Common Activatable Panel Legacy Object Reference)
   Event Construct()
   > Bind Event to Intro Ended(self:Lightbox Object Reference=Get Lightbox() -> (Lightbox:Lightbox Object Reference), Delegate=Create Event()) {is=title K2Node_AddDelegate}
   > Bind Event to Outro Ended(self:Lightbox Object Reference=Get Lightbox(), Delegate=Create Event()) {is=title K2Node_AddDelegate}
   > Branch(Condition=Get bAutoInitialize() -> (bAutoInitialize:bool))
   > InitializeModal(self:Fort Progress Modal Object Reference)
   Event OnBeginIntro() {is=title K2Node_Event}
   > n1 = Branch(Condition=Get bIntroOutroEnabled() -> (bIntroOutroEnabled:bool))
   > Intro(self:Lightbox Object Reference=Get Lightbox())
   EndIntro(execute=@n1.else)
   Event OnBeginOutro() {is=title K2Node_Event}
   > n2 = Branch(Condition=Get bIntroOutroEnabled())
   > Outro(self:Lightbox Object Reference=Get Lightbox())
   EndOutro(execute=@n2.else)
   Event Destruct()
   > Unbind Event from Intro Ended(self:Lightbox Object Reference=Get Lightbox(), Delegate=Create Event()) {is=title K2Node_RemoveDelegate}
   > Unbind Event from Outro Ended(self:Lightbox Object Reference=Get Lightbox(), Delegate=Create Event()) {is=title K2Node_RemoveDelegate}
  ]}
 OnAnalogValueChanged {nodes [
   OnAnalogValueChanged!() -> (then, MyGeometry:Geometry Structure, InAnalogInputEvent:Analog Input Event Structure) {is=title K2Node_FunctionEntry}
   > n1 = Branch(Condition=Get bConsumeAnalogInput() -> (bConsumeAnalogInput:bool))
   > Return Node!(execute, ReturnValue:Event Reply Structure=Handled())
   Return Node(execute=@n1.else, ReturnValue=Unhandled())
  ]}
]
variables []
widgets {
 info {
  inheritedWidgetCount=1
  namedSlotCount=1
  parentClass=/Script/FortniteUI.FortProgressModal
  rootWidgetClass=&Overlay
  widgetCount=8
 }
 widgets [
  Overlay_0 {widgetClassPath=&Overlay, widget=Overlay_0}
  BackgroundBlur_0 {
   widgetClassPath=&BackgroundBlur
   parent=Overlay_0
   widget=BackgroundBlur_0
   slot {class=OverlaySlot, properties {horizontalAlignment=HAlign_Fill, verticalAlignment=VAlign_Fill}}
  }
  Lightbox {
   widgetClassPath=/Game/UI/Foundation/Lightbox/Lightbox
   bIsVariable=true
   parent=BackgroundBlur_0
   widget=Lightbox
   slot {class=BackgroundBlurSlot, properties {padding=<ltrb 0 0 0 0>}}
  }
  Overlay_3 {widgetClassPath=&Overlay, namedSlotHost=Lightbox, widget=Overlay_3}
  SizeBox_1 {
   widgetClassPath=&SizeBox
   parent=Overlay_3
   widget=SizeBox_1
   properties {bOverride_MinDesiredHeight=true, bOverride_WidthOverride=true, minDesiredHeight=150, widthOverride=150}
   slot {class=OverlaySlot, properties {horizontalAlignment=HAlign_Center, verticalAlignment=VAlign_Center}}
  }
  ProgressSpinner {
   widgetClassPath=&Image
   bIsVariable=true
   parent=SizeBox_1
   widget=ProgressSpinner
   properties {brush {
     bIsDynamicallyLoaded=false
     drawAs=Image
     imageType=NoImage
     mirroring=NoMirror
     resourceName=None
     resourceObject=/FortUILibrary/UIKit/Throbber/MI_UIKit_Color_Throbber
     tiling=NoTile
     imageSize=<64 64>
     margin=<ltrb 0 0 0 0>
     outlineSettings {
      bUseBrushTransparency=false
      roundingType=HalfHeightRadius
      width=0
      color {colorUseRule=UseColor_Specified, specifiedColor=<rgba 0 0 0 0>}
      cornerRadii=<0 0 0 0>
     }
     tintColor {colorUseRule=UseColor_Specified, specifiedColor=<rgba 1 1 1 1>}
     uVRegion {bIsValid=false, max=<0 0>, min=<0 0>}
    }}
   slot {class=SizeBoxSlot}
  }
  SizeBox_0 {
   widgetClassPath=&SizeBox
   parent=Overlay_3
   widget=SizeBox_0
   properties {bOverride_MinDesiredHeight=true, bOverride_WidthOverride=true, minDesiredHeight=600, visibility=Collapsed, widthOverride=1000}
   slot {class=OverlaySlot, properties {horizontalAlignment=HAlign_Fill, verticalAlignment=VAlign_Fill}}
  }
  ProgressWidget_Main {
   widgetClassPath=/Game/UI/Foundation/Progress/ProgressWidget
   bInherited=true
   bIsVariable=true
   parent=SizeBox_0
   widget=ProgressWidget_Main
   properties {visibility=Collapsed}
   slot {class=SizeBoxSlot}
  }
 ]
}
