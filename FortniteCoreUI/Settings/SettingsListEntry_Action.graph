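# Settings list row widget for action-type settings, showing a setting name beside an action button, built on FortSettingsListEntrySetting_Action.
# When a setting is assigned, it sets the button's text and shows or hides its keybind widget.
# On non-touch input, mouse enter and leave play the OnHover animation forward and back, tinting the name text from white to dark blue.
# The action button is the primary gamepad focus target.
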WidgetBlueprint /FortniteCoreUI/Settings/SettingsListEntry_Action
animations [OnHover {
  binding_count=1
  length_seconds=0.1
  object=&self:OnHover
  sequence {bindings [Text_SettingName {
     id=#1
     possessed_class=CommonTextBlock
     tracks [{
       type=MovieSceneColorTrack
       display_name=Color and Opacity
       sections [{
         type=MovieSceneColorSection
         end_frame=2
         start_frame=0
         channels [
          Color.R {type=MovieSceneScriptingFloatChannel, keys [[0,1],[2,0.0009]]}
          Color.G {type=MovieSceneScriptingFloatChannel, keys [[0,1],[2,0.006]]}
          Color.B {type=MovieSceneScriptingFloatChannel, keys [[0,1],[2,0.0395]]}
          Color.A {type=MovieSceneScriptingFloatChannel, const [1,2]}
         ]
        }]
      }]
    }]}
 }]
graphs [
 EventGraph {nodes [
   n1 = Event OnSettingAssigned!() -> (OutputDelegate, then, ActionText:"Text (by ref)", bShowKeybindWidget:bool) {is=title K2Node_Event}
   > n2 = Cast To IconTextButton!(execute, Object=Get Button_Action() -> (Button_Action:Common Button Legacy Object Reference)) -> (then, CastFailed, AsIcon Text Button:Icon Text Button Object Reference) {is=title K2Node_DynamicCast}
   > Set Text!(execute, self:Icon Text Button Object Reference=@n2, ButtonText:Text=@n1.ActionText) -> (then) {is=title K2Node_CallFunction}
   > SetHideInputAction!(execute, self:Common Button Base Object Reference=Get Button_Action(), bInHideInputAction:bool=NOT Boolean(A=@n1.bShowKeybindWidget)) -> (then)
   Event OnMouseEnter()
   n3 = IsUsingTouch(PlayerController=GetOwningPlayer())
   > n4 = Branch(Condition=@n3)
   n5 = Get OnHover() -> (OnHover:Widget Animation Object Reference)
   PlayAnimationForward(execute=@n4.else, InAnimation=@n5)
   Event OnMouseLeave!() -> (OutputDelegate, then, MouseEvent:"Pointer Event Structure (by ref)") {is=title K2Node_Event}
   > n6 = Branch(Condition=@n3)
   PlayAnimationReverse(execute=@n6.else, InAnimation=@n5)
  ]}
 GetPrimaryGamepadFocusWidget {nodes [
   GetPrimaryGamepadFocusWidget() {is=title K2Node_FunctionEntry}
   > Return Node!(execute, ReturnValue:Widget Object Reference=Get Button_Action())
  ]}
]
variables []
widgets {
 info {
  inheritedWidgetCount=3
  namedSlotCount=1
  parentClass=/Script/FortniteUI.FortSettingsListEntrySetting_Action
  rootWidgetClass=&/SettingEntryBackground.SettingEntryBackground_C
  widgetCount=5
 }
 widgets [
  Background {widgetClassPath=&/SettingEntryBackground, bInherited=true, bIsVariable=true, widget=Background}
  FortResponsivePanel_0 {widgetClassPath=/Script/FortniteUI.FortResponsivePanel, namedSlotHost=Background, widget=FortResponsivePanel_0}
  SizeBox_0 {
   widgetClassPath=&SizeBox
   parent=FortResponsivePanel_0
   widget=SizeBox_0
   properties {bOverride_MinDesiredHeight=true, minDesiredHeight=50}
   slot {class=FortResponsivePanelSlot}
  }
  Text_SettingName {
   widgetClassPath=&CommonTextBlock
   bInherited=true
   bIsVariable=true
   parent=SizeBox_0
   widget=Text_SettingName
   properties {
    mobileFontSizeMultiplier=1.13
    scrollStyle=&TextScrollStyle-Base_C
    text="{Setting Name}"
    textTransformPolicy=ToUpper
    font {
     bForceMonospaced=false
     bMaterialIsStencil=false
     fontMaterial=null
     fontObject=&BurbankSmall
     letterSpacing=0
     monospacedWidth=1
     size=16
     skewAmount=0
     typefaceFontName=Black
     outlineSettings {
      bApplyOutlineToDropShadows=false
      bMiteredCorners=false
      bSeparateFillAlpha=false
      outlineMaterial=null
      outlineSize=0
      outlineColor=<rgba 0 0 0 1>
     }
    }
    margin=<ltrb 40 0 0 0>
    shadowOffset=<0 0>
   }
   slot {class=SizeBoxSlot, properties {verticalAlignment=VAlign_Center}}
  }
  Button_Action {
   widgetClassPath=&IconTextButton
   bInherited=true
   bIsVariable=true
   parent=FortResponsivePanel_0
   widget=Button_Action
   properties {
    buttonText="{Action}"
    contentAlignment=HAlign_Center
    onButtonBaseClicked=(&self:WidgetTree.Button_Action.HandleOnButtonClicked)
    onButtonBaseDoubleClicked=(&self:WidgetTree.Button_Action.HandleOnButtonDoubleClicked)
    onButtonBaseHovered=(&self:WidgetTree.Button_Action.HandleOnButtonHovered)
    onButtonBaseUnhovered=(&self:WidgetTree.Button_Action.HandleOnButtonUnhovered)
    onSelectedChangedBase=(&self:WidgetTree.Button_Action.HandleOnSelectedChanged)
    style=/FortUILibrary/Foundation/Buttons/ButtonStyles/Athena/Athena_ButtonStyle_BlueMenuButton_Settings.Athena_ButtonStyle_BlueMenuButton_Settings_C
    touchMethod=PreciseTap
   }
   slot {class=FortResponsivePanelSlot}
  }
 ]
}
