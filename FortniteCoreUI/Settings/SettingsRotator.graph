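# Settings rotator widget, based on FortSettingsRotator, that shows the current option of a setting as centered, uppercase text inside a material-backed Border.
# Parent class events push the option count, selected index and default index into scalar parameters on the Background's dynamic SettingsRotator_M material, leaving the option visuals to the material.
# SetTextScrolling toggles scrolling on MyText and is called on initialization to turn it off, with overflow set to ellipsis.
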WidgetBlueprint /FortniteCoreUI/Settings/SettingsRotator
graphs [
 EventGraph {nodes [
   n1 = Event On Options Populated!() -> (OutputDelegate, then, Count:int=0) {is=title K2Node_Event}
   > n2 = GetDynamicMaterial(self:Border Object Reference=Get Background() -> (Background:Border Object Reference))
   > SetScalarParameterValue(self=@n2, ParameterName=Options, Value="To Float (Integer)"(InInt=@n1.Count))
   n3 = Event On Options Selected!() -> (OutputDelegate, then, Index:int=0) {is=title K2Node_Event}
   > n4 = GetDynamicMaterial(self=Get Background())
   > SetScalarParameterValue(self=@n4, ParameterName=SelectedOption, Value="To Float (Integer)"(InInt=@n3.Index))
   n5 = Event On Default Option Specified!() -> (OutputDelegate, then, DefaultOptionIndex:int=0) {is=title K2Node_Event}
   > n6 = GetDynamicMaterial(self=Get Background())
   > SetScalarParameterValue(self=@n6, ParameterName=DefaultOption, Value="To Float (Integer)"(InInt=@n5.DefaultOptionIndex))
   Event OnInitialized()
   > SetTextScrolling!(execute, self, bIsScrollingEnabled:bool=false) -> (then)
  ]}
 SetTextScrolling {nodes [
   n1 = SetTextScrolling!() -> (then, bIsScrollingEnabled:bool=~) {is=title K2Node_FunctionEntry}
   > SetScrollingEnabled!(execute, self:Common Text Object Reference=Get MyText() -> (MyText:Common Text Object Reference), bInIsScrollingEnabled:bool=@n1) -> (then)
  ]}
]
variables []
widgets {
 info {inheritedWidgetCount=2, parentClass=/Script/FortniteUI.FortSettingsRotator, rootWidgetClass=&Border, widgetCount=3}
 widgets [
  Background {
   widgetClassPath=&Border
   bIsVariable=true
   widget=Background
   properties {
    horizontalAlignment=HAlign_Center
    verticalAlignment=VAlign_Center
    background {
     bIsDynamicallyLoaded=false
     drawAs=Image
     imageType=NoImage
     mirroring=NoMirror
     resourceName=None
     resourceObject=/FortniteCoreUI/Settings/Art/SettingsRotator_M
     tiling=NoTile
     imageSize=<32 32>
     margin=<ltrb 0 0 0 0>
     outlineSettings {
      bUseBrushTransparency=false
      roundingType=HalfHeightRadius
      width=0
      color {colorUseRule=UseColor_Specified, specifiedColor=<rgba 0 0 0 0>}
      cornerRadii=<0 0 0 0>
     }
     tintColor {colorUseRule=UseColor_Specified, specifiedColor=<rgba 1 1 1 1>}
     uVRegion {bIsValid=false, max=<0 0>, min=<0 0>}
    }
    padding=<ltrb 8 2 8 4>
   }
  }
  MyText {
   widgetClassPath=&CommonTextBlock
   bInherited=true
   bIsVariable=true
   parent=Background
   widget=MyText
   properties {
    bIsScrollingEnabled=false
    clipping=ClipToBounds
    mobileFontSizeMultiplier=1.14
    scrollStyle=&TextScrollStyle-Base_C
    text="{Option}"
    textOverflowPolicy=Ellipsis
    textTransformPolicy=ToUpper
    colorAndOpacity {colorUseRule=UseColor_Specified, specifiedColor=<rgba 0.6105 0.8796 1 1>}
    font {
     bForceMonospaced=false
     bMaterialIsStencil=false
     fontMaterial=null
     fontObject=&BurbankSmall
     letterSpacing=0
     monospacedWidth=1
     size=15.75
     skewAmount=0
     typefaceFontName=Black
     outlineSettings {
      bApplyOutlineToDropShadows=false
      bMiteredCorners=false
      bSeparateFillAlpha=false
      outlineMaterial=null
      outlineSize=0
      outlineColor=<rgba 0 0 0 1>
     }
    }
    shadowOffset=<0 0>
   }
   slot {class=BorderSlot, properties {horizontalAlignment=HAlign_Center, verticalAlignment=VAlign_Center, padding=<ltrb 8 2 8 4>}}
  }
  InputActionWidget {widgetClassPath=&CommonActionWidget, bInherited=true, widget=null}
 ]
}
