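# Settings menu row for a discrete option, built on FortSettingsListEntrySetting_Discrete with the setting name and a value panel.
# A switcher holds a two-button layout and an active layout of decrease and increase arrows around a SettingsRotator, which takes gamepad focus.
# Focus changes toggle the rotator's text scrolling.
# On non-touch input, mouse hover also toggles it, animates the name's color forward or back, and plays a sound on enter.
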WidgetBlueprint /FortniteCoreUI/Settings/SettingsListEntry_Discrete
animations [OnHover {
  binding_count=1
  length_seconds=0.1
  object=&self:OnHover
  sequence {
   bindings [Text_SettingName {
     id=#1
     possessed_class=CommonTextBlock
     tracks [{
       type=MovieSceneColorTrack
       display_name=Color and Opacity
       sections [{
         type=MovieSceneColorSection
         end_frame=12
         start_frame=0
         channels [
          Color.R {type=MovieSceneScriptingFloatChannel, keys [[0,1],[12,0.0009]]}
          Color.G {type=MovieSceneScriptingFloatChannel, keys [[0,1],[12,0.006]]}
          Color.B {type=MovieSceneScriptingFloatChannel, keys [[0,1],[12,0.0395]]}
          Color.A {type=MovieSceneScriptingFloatChannel, const [1,2]}
         ]
        }]
      }]
    }]
   display_rate {numerator=120}
  }
 }]
graphs [
 EventGraph {nodes [
   Event OnAddedToFocusPath!() -> (OutputDelegate, then, InFocusEvent:Focus Event Structure) {is=title K2Node_Event}
   > SetSettingsRotatorTextScrolling!(execute, self, bIsScrollingEnabled:bool=true) -> (then)
   Event OnRemovedFromFocusPath!() -> (OutputDelegate, then, InFocusEvent:Focus Event Structure) {is=title K2Node_Event}
   > SetSettingsRotatorTextScrolling(bIsScrollingEnabled=false)
   Event OnMouseEnter()
   n1 = IsUsingTouch(PlayerController=GetOwningPlayer())
   > n2 = Branch(Condition=@n1)
   n3 = Get OnHover() -> (OnHover:Widget Animation Object Reference)
   PlayAnimationForward(execute=@n2.else, InAnimation=@n3)
   > SetSettingsRotatorTextScrolling(bIsScrollingEnabled=true)
   > PlaySound2D(Sound=/Game/Sounds/Fort_UI_Sounds/FrontEnd/DefaultSounds/PressedButtons/MSS_UI_Button_Hover_Default)
   Event OnMouseLeave!() -> (OutputDelegate, then, MouseEvent:"Pointer Event Structure (by ref)") {is=title K2Node_Event}
   > n4 = Branch(Condition=@n1)
   PlayAnimationReverse(execute=@n4.else, InAnimation=@n3)
   > SetSettingsRotatorTextScrolling(bIsScrollingEnabled=false)
  ]}
 GetPrimaryGamepadFocusWidget {nodes [
   GetPrimaryGamepadFocusWidget() {is=title K2Node_FunctionEntry}
   > Return Node!(execute, ReturnValue:Widget Object Reference=Get Rotator_SettingValue() -> (Rotator_SettingValue:Fort Settings Rotator Object Reference))
  ]}
 SetSettingsRotatorTextScrolling {nodes [
   n1 = SetSettingsRotatorTextScrolling!() -> (then, bIsScrollingEnabled:bool=~) {is=title K2Node_FunctionEntry}
   > n2 = Cast To SettingsRotator!(execute, Object=Get Rotator_SettingValue()) -> (then, CastFailed, AsSettings Rotator:Settings Rotator Object Reference) {is=title K2Node_DynamicCast}
   > SetTextScrolling!(execute, self:Settings Rotator Object Reference=@n2, bIsScrollingEnabled:bool=@n1) -> (then)
  ]}
]
variables []
widgets {
 info {
  inheritedWidgetCount=8
  namedSlotCount=1
  parentClass=/Script/FortniteUI.FortSettingsListEntrySetting_Discrete
  rootWidgetClass=&/SettingEntryBackground.SettingEntryBackground_C
  widgetCount=16
 }
 widgets [
  Background {widgetClassPath=&/SettingEntryBackground, bInherited=true, bIsVariable=true, widget=Background}
  FortResponsivePanel_85 {widgetClassPath=/Script/FortniteUI.FortResponsivePanel, namedSlotHost=Background, widget=FortResponsivePanel_85}
  SizeBox_0 {
   widgetClassPath=&SizeBox
   parent=FortResponsivePanel_85
   widget=SizeBox_0
   properties {bOverride_MinDesiredHeight=true, minDesiredHeight=50}
   slot {class=FortResponsivePanelSlot}
  }
  Text_SettingName {
   widgetClassPath=&CommonTextBlock
   bInherited=true
   bIsVariable=true
   parent=SizeBox_0
   widget=Text_SettingName
   properties {
    bIsScrollingEnabled=false
    clipping=ClipToBounds
    mobileFontSizeMultiplier=1.13
    text="{Setting Name}"
    textOverflowPolicy=Ellipsis
    textTransformPolicy=ToUpper
    font {
     bForceMonospaced=false
     bMaterialIsStencil=false
     fontMaterial=null
     fontObject=&BurbankSmall
     letterSpacing=0
     monospacedWidth=1
     size=16
     skewAmount=0
     typefaceFontName=Black
     outlineSettings {
      bApplyOutlineToDropShadows=false
      bMiteredCorners=false
      bSeparateFillAlpha=false
      outlineMaterial=null
      outlineSize=0
      outlineColor=<rgba 0 0 0 1>
     }
    }
    margin=<ltrb 40 0 0 0>
    shadowOffset=<0 0>
   }
   slot {class=SizeBoxSlot, properties {verticalAlignment=VAlign_Center}}
  }
  Panel_Value {
   widgetClassPath=&HorizontalBox
   bInherited=true
   bIsVariable=true
   parent=FortResponsivePanel_85
   widget=Panel_Value
   slot {class=FortResponsivePanelSlot}
  }
  OptionsSets {
   widgetClassPath=&WidgetSwitcher
   bIsVariable=true
   parent=Panel_Value
   widget=OptionsSets
   properties {activeWidgetIndex=1}
   slot {class=HorizontalBoxSlot, properties {size {sizeRule=Fill, value=1}}}
  }
  Couple_Options {widgetClassPath=&HorizontalBox, parent=OptionsSets, widget=Couple_Options, slot {class=WidgetSwitcherSlot}}
  Button_FirstOption {
   widgetClassPath=&IconTextButton
   bIsVariable=true
   parent=Couple_Options
   widget=Button_FirstOption
   properties {
    bInteractableWhenSelected=true
    bToggleable=true
    buttonText=""
    onButtonBaseClicked=(&self:WidgetTree.Button_FirstOption.HandleOnButtonClicked)
    onButtonBaseDoubleClicked=(&self:WidgetTree.Button_FirstOption.HandleOnButtonDoubleClicked)
    onButtonBaseHovered=(&self:WidgetTree.Button_FirstOption.HandleOnButtonHovered)
    onButtonBaseUnhovered=(&self:WidgetTree.Button_FirstOption.HandleOnButtonUnhovered)
    onSelectedChangedBase=(&self:WidgetTree.Button_FirstOption.HandleOnSelectedChanged)
    style=/FortUILibrary/Foundation/Buttons/ButtonStyles/Athena/Athena_ButtonStyle_BlueMenuButton_Settings.Athena_ButtonStyle_BlueMenuButton_Settings_C
   }
   slot {class=HorizontalBoxSlot, properties {size {sizeRule=Fill, value=1}}}
  }
  Button_SecondOption {
   widgetClassPath=&IconTextButton
   bIsVariable=true
   parent=Couple_Options
   widget=Button_SecondOption
   properties {
    bInteractableWhenSelected=true
    bToggleable=true
    buttonText=""
    onButtonBaseClicked=(&self:WidgetTree.Button_SecondOption.HandleOnButtonClicked)
    onButtonBaseDoubleClicked=(&self:WidgetTree.Button_SecondOption.HandleOnButtonDoubleClicked)
    onButtonBaseHovered=(&self:WidgetTree.Button_SecondOption.HandleOnButtonHovered)
    onButtonBaseUnhovered=(&self:WidgetTree.Button_SecondOption.HandleOnButtonUnhovered)
    onSelectedChangedBase=(&self:WidgetTree.Button_SecondOption.HandleOnSelectedChanged)
    style=&Athena_ButtonStyle_BlueMenuButton_Settings_C
    renderTransform {angle=0, scale=<-1 1>, shear=<0 0>, translation=<0 0>}
   }
   slot {class=HorizontalBoxSlot, properties {size {sizeRule=Fill, value=1}}}
  }
  Several_Options {widgetClassPath=&HorizontalBox, bIsVariable=true, parent=OptionsSets, widget=Several_Options, slot {class=WidgetSwitcherSlot}}
  Button_Decrease {
   widgetClassPath=&IconTextButton
   bInherited=true
   bIsVariable=true
   parent=Several_Options
   widget=Button_Decrease
   properties {
    bInteractableWhenSelected=true
    bToggleable=true
    buttonText=""
    onButtonBaseClicked=(&self:WidgetTree.Button_Decrease.HandleOnButtonClicked)
    onButtonBaseDoubleClicked=(&self:WidgetTree.Button_Decrease.HandleOnButtonDoubleClicked)
    onButtonBaseHovered=(&self:WidgetTree.Button_Decrease.HandleOnButtonHovered)
    onButtonBaseUnhovered=(&self:WidgetTree.Button_Decrease.HandleOnButtonUnhovered)
    onSelectedChangedBase=(&self:WidgetTree.Button_Decrease.HandleOnSelectedChanged)
    style=&Athena_ButtonStyle_BlueMenuButton_Settings_C
    touchMethod=PreciseTap
    iconBrush {
     bIsDynamicallyLoaded=false
     drawAs=Image
     imageType=NoImage
     mirroring=NoMirror
     resourceName=None
     resourceObject=/Game/UI/SocialPanel/Art/Textures/T_UI_Icon_TreeArrow_64
     tiling=NoTile
     imageSize=<32 32>
     margin=<ltrb 0 0 0 0>
     outlineSettings {
      bUseBrushTransparency=false
      roundingType=HalfHeightRadius
      width=0
      color {colorUseRule=UseColor_Specified, specifiedColor=<rgba 0 0 0 0>}
      cornerRadii=<0 0 0 1>
     }
     tintColor {colorUseRule=UseColor_Specified, specifiedColor=<rgba 0.0331 0.402 1 1>}
     uVRegion {bIsValid=false, max=<0 0>, min=<0 0>}
    }
    renderTransform {angle=0, scale=<-1 1>, shear=<0 0>, translation=<0 0>}
   }
   slot {class=HorizontalBoxSlot}
  }
  Border_0 {
   widgetClassPath=&Border
   parent=Several_Options
   widget=Border_0
   properties {clipping=ClipToBounds, brushColor=<rgba 0.0033 0.0097 0.1022 1>, padding=<ltrb 0 0 0 0>}
   slot {class=HorizontalBoxSlot, properties {size {sizeRule=Fill, value=1}}}
  }
  Rotator_SettingValue {
   widgetClassPath=&/SettingsRotator
   bInherited=true
   bIsVariable=true
   parent=Border_0
   widget=Rotator_SettingValue
   properties {clipping=ClipToBounds, touchMethod=PreciseTap}
   slot {class=BorderSlot, properties {padding=<ltrb 0 0 0 0>}}
  }
  Button_Increase {
   widgetClassPath=&IconTextButton
   bInherited=true
   bIsVariable=true
   parent=Several_Options
   widget=Button_Increase
   properties {
    bInteractableWhenSelected=true
    bToggleable=true
    buttonText=""
    onButtonBaseClicked=(&self:WidgetTree.Button_Increase.HandleOnButtonClicked)
    onButtonBaseDoubleClicked=(&self:WidgetTree.Button_Increase.HandleOnButtonDoubleClicked)
    onButtonBaseHovered=(&self:WidgetTree.Button_Increase.HandleOnButtonHovered)
    onButtonBaseUnhovered=(&self:WidgetTree.Button_Increase.HandleOnButtonUnhovered)
    onSelectedChangedBase=(&self:WidgetTree.Button_Increase.HandleOnSelectedChanged)
    style=&Athena_ButtonStyle_BlueMenuButton_Settings_C
    touchMethod=PreciseTap
    iconBrush {
     bIsDynamicallyLoaded=false
     drawAs=Image
     imageType=NoImage
     mirroring=NoMirror
     resourceName=None
     resourceObject=&T_UI_Icon_TreeArrow_64
     tiling=NoTile
     imageSize=<32 32>
     margin=<ltrb 0 0 0 0>
     outlineSettings {
      bUseBrushTransparency=false
      roundingType=HalfHeightRadius
      width=0
      color {colorUseRule=UseColor_Specified, specifiedColor=<rgba 0 0 0 0>}
      cornerRadii=<0 0 0 1>
     }
     tintColor {colorUseRule=UseColor_Specified, specifiedColor=<rgba 0.0331 0.402 1 1>}
     uVRegion {bIsValid=false, max=<0 0>, min=<0 0>}
    }
   }
   slot {class=HorizontalBoxSlot}
  }
  Button_SelectEntry {widgetClassPath=&CommonButtonLegacy, bInherited=true, widget=null}
  OptionsModifiedIndicator {widgetClassPath=&Image, bInherited=true, widget=null}
 ]
}
