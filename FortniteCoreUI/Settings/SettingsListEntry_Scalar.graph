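# Settings list row widget for scalar settings, extending FortSettingsListEntrySetting_Scalar with an uppercase name beside a slider-style value display.
# An invisible AnalogSlider overlays a material progress bar, default-value image and value text, and serves as the gamepad focus target.
# Value changes drive the bar, while default changes set the DefaultValue parameter on the image's dynamic material.
# Mouse hover animates the name's color and plays a sound on enter, skipped for touch.
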WidgetBlueprint /FortniteCoreUI/Settings/SettingsListEntry_Scalar
animations [OnHover {
  binding_count=1
  length_seconds=0.1
  object=&self:OnHover
  sequence {
   bindings [Text_SettingName {
     id=#1
     possessed_class=CommonTextBlock
     tracks [{
       type=MovieSceneColorTrack
       display_name=Color and Opacity
       sections [{
         type=MovieSceneColorSection
         end_frame=12
         start_frame=0
         channels [
          Color.R {type=MovieSceneScriptingFloatChannel, keys [[0,1],[12,0.0009]]}
          Color.G {type=MovieSceneScriptingFloatChannel, keys [[0,1],[12,0.006]]}
          Color.B {type=MovieSceneScriptingFloatChannel, keys [[0,1],[12,0.0395]]}
          Color.A {type=MovieSceneScriptingFloatChannel, const [1,2]}
         ]
        }]
      }]
    }]
   display_rate {numerator=120}
  }
 }]
graphs [
 EventGraph {nodes [
   n1 = Event OnValueChanged!() -> (OutputDelegate, then, Value:float) {is=title K2Node_Event}
   > SetProgress!(
    execute
    self:Fort Simple Material Progress Bar Object Reference=Get Progress() -> (Progress:Simple Material Progress Bar Object Reference)
    Progress:float=@n1.Value
   ) -> (then)
   n2 = Event OnDefaultValueChanged!() -> (OutputDelegate, then, DefaultValue:float) {is=title K2Node_Event}
   > n3 = GetDynamicMaterial(self=Get DefaultValueLayer() -> (DefaultValueLayer:Image Object Reference))
   > SetScalarParameterValue(self=@n3, ParameterName=DefaultValue, Value=@n2.DefaultValue)
   Event OnMouseEnter()
   n4 = IsUsingTouch(PlayerController=GetOwningPlayer())
   > n5 = Branch(Condition=@n4)
   n6 = Get OnHover() -> (OnHover:Widget Animation Object Reference)
   PlayAnimationForward(execute=@n5.else, InAnimation=@n6)
   > PlaySound2D(Sound=/Game/Sounds/Fort_UI_Sounds/FrontEnd/DefaultSounds/PressedButtons/MSS_UI_Button_Hover_Default)
   Event OnMouseLeave!() -> (OutputDelegate, then, MouseEvent:"Pointer Event Structure (by ref)") {is=title K2Node_Event}
   > n7 = Branch(Condition=@n4)
   PlayAnimationReverse(execute=@n7.else, InAnimation=@n6)
  ]}
 GetPrimaryGamepadFocusWidget {nodes [
   GetPrimaryGamepadFocusWidget() {is=title K2Node_FunctionEntry}
   > Return Node!(execute, ReturnValue:Widget Object Reference=Get Slider_SettingValue() -> (Slider_SettingValue:Analog Slider Object Reference))
  ]}
]
variables []
widgets {
 info {
  inheritedWidgetCount=5
  namedSlotCount=1
  parentClass=/Script/FortniteUI.FortSettingsListEntrySetting_Scalar
  rootWidgetClass=&/SettingEntryBackground.SettingEntryBackground_C
  widgetCount=10
 }
 widgets [
  Background {widgetClassPath=&/SettingEntryBackground, bInherited=true, bIsVariable=true, widget=Background}
  FortResponsivePanel_1 {widgetClassPath=/Script/FortniteUI.FortResponsivePanel, namedSlotHost=Background, widget=FortResponsivePanel_1}
  SizeBox_0 {
   widgetClassPath=&SizeBox
   parent=FortResponsivePanel_1
   widget=SizeBox_0
   properties {bOverride_MinDesiredHeight=true, minDesiredHeight=50}
   slot {class=FortResponsivePanelSlot}
  }
  Text_SettingName {
   widgetClassPath=&CommonTextBlock
   bInherited=true
   bIsVariable=true
   parent=SizeBox_0
   widget=Text_SettingName
   properties {
    scrollStyle=&TextScrollStyle-Base_C
    text="{Setting Name}"
    textTransformPolicy=ToUpper
    font {
     bForceMonospaced=false
     bMaterialIsStencil=false
     fontMaterial=null
     fontObject=&BurbankSmall
     letterSpacing=0
     monospacedWidth=1
     size=16
     skewAmount=0
     typefaceFontName=Black
     outlineSettings {
      bApplyOutlineToDropShadows=false
      bMiteredCorners=false
      bSeparateFillAlpha=false
      outlineMaterial=null
      outlineSize=0
      outlineColor=<rgba 0 0 0 1>
     }
    }
    margin=<ltrb 40 0 0 0>
    shadowOffset=<0 0>
   }
   slot {class=SizeBoxSlot, properties {verticalAlignment=VAlign_Center}}
  }
  SizeBox_1 {
   widgetClassPath=&SizeBox
   parent=FortResponsivePanel_1
   widget=SizeBox_1
   properties {bOverride_MinDesiredHeight=true, minDesiredHeight=50}
   slot {class=FortResponsivePanelSlot}
  }
  Panel_Value {widgetClassPath=&Overlay, bInherited=true, bIsVariable=true, parent=SizeBox_1, widget=Panel_Value, slot {class=SizeBoxSlot}}
  Progress {
   widgetClassPath=/Game/Athena/HUD/SimpleMaterialProgressBar
   bIsVariable=true
   parent=Panel_Value
   widget=Progress
   properties {
    bOverrideDefaultColorA=true
    bOverrideDefaultColorB=true
    bOverrideDefaultColorBackground=true
    visibility=HitTestInvisible
    cachedColorA=<rgba 0.0382 0.4452 0.9047 1>
    cachedColorB=<rgba 0.0382 0.4452 0.9047 1>
    cachedColorBackground=<rgba 0.0033 0.0097 0.1022 1>
   }
   slot {class=OverlaySlot, properties {horizontalAlignment=HAlign_Fill, verticalAlignment=VAlign_Fill}}
  }
  DefaultValueLayer {
   widgetClassPath=&Image
   bIsVariable=true
   parent=Panel_Value
   widget=DefaultValueLayer
   properties {brush {
     bIsDynamicallyLoaded=false
     drawAs=Image
     imageType=NoImage
     mirroring=NoMirror
     resourceName=None
     resourceObject=/Game/Athena/Settings/Art/SettingsDefaultSlider_M
     tiling=NoTile
     imageSize=<32 32>
     margin=<ltrb 0 0 0 0>
     outlineSettings {
      bUseBrushTransparency=false
      roundingType=HalfHeightRadius
      width=0
      color {colorUseRule=UseColor_Specified, specifiedColor=<rgba 0 0 0 0>}
      cornerRadii=<0 0 0 0>
     }
     tintColor {colorUseRule=UseColor_Specified, specifiedColor=<rgba 1 1 1 1>}
     uVRegion {bIsValid=false, max=<0 0>, min=<0 0>}
    }}
   slot {class=OverlaySlot, properties {horizontalAlignment=HAlign_Fill, verticalAlignment=VAlign_Fill}}
  }
  Text_SettingValue {
   widgetClassPath=&CommonTextBlock
   bInherited=true
   bIsVariable=true
   parent=Panel_Value
   widget=Text_SettingValue
   properties {
    justification=Center
    scrollStyle=&TextScrollStyle-Base_C
    style=/FortUILibrary/Styles/Typography/Foundation/TextStyle-Base-S-B.TextStyle-Base-S-B_C
    text="{Value}"
    visibility=HitTestInvisible
    font {
     bForceMonospaced=false
     bMaterialIsStencil=false
     fontMaterial=null
     fontObject=&NotoSans
     letterSpacing=0
     monospacedWidth=1
     size=18
     skewAmount=0
     typefaceFontName=NotoSans-Bold
     outlineSettings {
      bApplyOutlineToDropShadows=false
      bMiteredCorners=false
      bSeparateFillAlpha=false
      outlineMaterial=null
      outlineSize=0
      outlineColor=<rgba 0 0 0 1>
     }
    }
    shadowOffset=<0 0>
   }
   slot {class=OverlaySlot, properties {horizontalAlignment=HAlign_Fill, verticalAlignment=VAlign_Center}}
  }
  Slider_SettingValue {
   widgetClassPath=/Script/CommonUI.AnalogSlider
   bInherited=true
   bIsVariable=true
   parent=Panel_Value
   widget=Slider_SettingValue
   properties {sliderBarColor=<rgba 1 1 1 0>, sliderHandleColor=<rgba 1 1 1 0>}
   slot {class=OverlaySlot, properties {horizontalAlignment=HAlign_Fill, verticalAlignment=VAlign_Fill}}
  }
 ]
}
