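# Settings list row widget, based on ListEntrySettingTimeWindow, for a daily time window setting with a name label and an edit button.
# DisplayTimeWindow breaks the window into start and end times and shows them as a range on the button.
# Clicking the button reports the setting as selected, and the button also serves as the gamepad focus target.
# Mouse enter and leave play a label color animation forward and back, skipped when using touch.
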WidgetBlueprint /FortniteCoreUI/Settings/SettingsListEntry_TimeWindow
animations [OnHover {
  binding_count=1
  length_seconds=0.1
  object=&self:OnHover
  sequence {bindings [Text_SettingName {
     id=#1
     possessed_class=CommonTextBlock
     tracks [{
       type=MovieSceneColorTrack
       display_name=Color and Opacity
       sections [{
         type=MovieSceneColorSection
         end_frame=2
         start_frame=0
         channels [
          Color.R {type=MovieSceneScriptingFloatChannel, keys [[0,1],[2,0.0009]]}
          Color.G {type=MovieSceneScriptingFloatChannel, keys [[0,1],[2,0.006]]}
          Color.B {type=MovieSceneScriptingFloatChannel, keys [[0,1],[2,0.0395]]}
          Color.A {type=MovieSceneScriptingFloatChannel, const [1,2]}
         ]
        }]
      }]
    }]}
 }]
graphs [
 EventGraph {nodes [
   "On Button Clicked (Button_EditWindow)"!() -> (OutputDelegate, then, Button:Common Button Legacy Object Reference) {is=title K2Node_ComponentBoundEvent}
   > NotifySettingSelected(self:List Entry Setting Time Window Object Reference)
   n1 = Event DisplayTimeWindow!() -> (OutputDelegate, then, TimeWindow:"Daily Time Window Structure (by ref)") {is=title K2Node_Event}
   n2 = Break Daily Time Window!(DailyTimeWindow:"Daily Time Window Structure (by ref)"=@n1.TimeWindow) -> (StartTime:Daily Time Structure, EndTime:Daily Time Structure) {is=title K2Node_BreakStruct}
   > Set Text!(
    execute
    self:Icon Text Button Object Reference=Get Button_EditWindow() -> (Button_EditWindow:Icon Text Button Object Reference)
    ButtonText:Text=Format Text!(
     Format="INVTEXT(\"{0} - {1}\")"
     0:Text=ToText!(DailyTime:"Daily Time Structure (by ref)"=@n2) -> (ReturnValue:Text)
     1:Text=ToText(DailyTime=@n2.EndTime)
    ) -> (Result)
   ) -> (then) {is=title K2Node_CallFunction}
   Event OnMouseEnter()
   n3 = IsUsingTouch(PlayerController=GetOwningPlayer())
   > n4 = Branch(Condition=@n3)
   n5 = Get OnHover() -> (OnHover:Widget Animation Object Reference)
   PlayAnimationForward(execute=@n4.else, InAnimation=@n5)
   Event OnMouseLeave!() -> (OutputDelegate, then, MouseEvent:"Pointer Event Structure (by ref)") {is=title K2Node_Event}
   > n6 = Branch(Condition=@n3)
   PlayAnimationReverse(execute=@n6.else, InAnimation=@n5)
  ]}
 GetPrimaryGamepadFocusWidget {nodes [
   GetPrimaryGamepadFocusWidget() {is=title K2Node_FunctionEntry}
   > Return Node!(execute, ReturnValue:Widget Object Reference=Get Button_EditWindow())
  ]}
]
variables []
widgets {
 info {
  inheritedWidgetCount=0
  namedSlotCount=1
  parentClass=/Script/FortniteUI.ListEntrySettingTimeWindow
  rootWidgetClass=&/SettingEntryBackground.SettingEntryBackground_C
  widgetCount=5
 }
 widgets [
  Background {widgetClassPath=&/SettingEntryBackground, bIsVariable=true, widget=Background}
  FortResponsivePanel_0 {widgetClassPath=/Script/FortniteUI.FortResponsivePanel, namedSlotHost=Background, widget=FortResponsivePanel_0}
  SizeBox_0 {
   widgetClassPath=&SizeBox
   parent=FortResponsivePanel_0
   widget=SizeBox_0
   properties {bOverride_MinDesiredHeight=true, minDesiredHeight=50}
   slot {class=FortResponsivePanelSlot}
  }
  Text_SettingName {
   widgetClassPath=&CommonTextBlock
   bIsVariable=true
   parent=SizeBox_0
   widget=Text_SettingName
   properties {
    mobileFontSizeMultiplier=1.13
    scrollStyle=&TextScrollStyle-Base_C
    text=Time Window
    textTransformPolicy=ToUpper
    font {
     bForceMonospaced=false
     bMaterialIsStencil=false
     fontMaterial=null
     fontObject=&BurbankSmall
     letterSpacing=0
     monospacedWidth=1
     size=16
     skewAmount=0
     typefaceFontName=Black
     outlineSettings {
      bApplyOutlineToDropShadows=false
      bMiteredCorners=false
      bSeparateFillAlpha=false
      outlineMaterial=null
      outlineSize=0
      outlineColor=<rgba 0 0 0 1>
     }
    }
    margin=<ltrb 40 0 0 0>
    shadowOffset=<0 0>
   }
   slot {class=SizeBoxSlot, properties {verticalAlignment=VAlign_Center}}
  }
  Button_EditWindow {
   widgetClassPath=&IconTextButton
   bIsVariable=true
   parent=FortResponsivePanel_0
   widget=Button_EditWindow
   properties {
    buttonText=10:00 AM - 1:00 PM
    contentAlignment=HAlign_Center
    onButtonBaseClicked=(&self:WidgetTree.Button_EditWindow.HandleOnButtonClicked)
    onButtonBaseDoubleClicked=(&self:WidgetTree.Button_EditWindow.HandleOnButtonDoubleClicked)
    onButtonBaseHovered=(&self:WidgetTree.Button_EditWindow.HandleOnButtonHovered)
    onButtonBaseUnhovered=(&self:WidgetTree.Button_EditWindow.HandleOnButtonUnhovered)
    onSelectedChangedBase=(&self:WidgetTree.Button_EditWindow.HandleOnSelectedChanged)
    style=/FortUILibrary/Foundation/Buttons/ButtonStyles/Athena/Athena_ButtonStyle_BlueMenuButton_Settings.Athena_ButtonStyle_BlueMenuButton_Settings_C
    touchMethod=PreciseTap
   }
   slot {class=FortResponsivePanelSlot}
  }
 ]
}
