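# Settings list entry widget for a time-windows setting, showing a vertical box of window rows above an "Add Another Window" button.
# The button calls AddWindow and its background shows only while CanAddNewWindow allows.
# UI refresh chains explicit up/down navigation through rows and button, escaping at the ends.
# Gamepad focus returns to the last selected window, or picks top or bottom by comparing the settings panel's selection indices with SettingIndex.
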WidgetBlueprint /FortniteCoreUI/Settings/SettingsListEntry_TimeWindows
graphs [
 EventGraph {nodes [
   "On Button Clicked (Button_AddWindow)"!() -> (OutputDelegate, then, Button:Common Button Legacy Object Reference) {is=title K2Node_ComponentBoundEvent}
   > AddWindow(self:Fort Settings List Entry Setting Time Windows Object Reference)
   Event OnRefreshUI() {is=title K2Node_Event}
   > SetVisibility(
    self=Get Background_AddWindow() -> (Background_AddWindow:Setting Entry Background Object Reference)
    InVisibility=Select(
     Option 0=Collapsed
     Option 1=Visible
     Index=CanAddNewWindow!(self:Fort Settings List Entry Setting Time Windows Object Reference) -> (ReturnValue:bool)
    )
   )
   > n1 = For Each Loop(Array:Array of User Widget Object References=GetAllEntries!(self:Dynamic Entry Box Base Object Reference=Get EntryBox_TimeWindows() -> (EntryBox_TimeWindows:Dynamic Entry Box Object Reference)) -> (ReturnValue:Array of User Widget Object References)) -> (Array Element:User Widget Object Reference)
   n2 = Get EntryBox_TimeWindows()
   Branch(
    execute=@n1.Completed
    Condition="Equal (Integer)"(A=GetNumEntries!(self:Dynamic Entry Box Base Object Reference=@n2) -> (ReturnValue:int=0))
   )
   n3 = Get Button_AddWindow() -> (Button_AddWindow:Icon Text Button Object Reference)
   > SetNavigationRuleBase!(execute, self:Widget Object Reference=@n3, Direction:EUINavigation Enum=Up, Rule:EUINavigationRule Enum=Escape) -> (then)
   > SetNavigationRuleBase(self=@n3, Direction=Down, Rule=Escape)
   Branch(execute=@n1.LoopBody, Condition=IsValid(Object=@n1))
   > n4 = Branch(Condition="Equal (Integer)"(A=@n1.Array Index))
   > n5 = SetNavigationRuleBase(self=@n1, Direction=Up, Rule=Escape)
   n6 = int + int(A=@n1.Array Index)
   n7 = "Get (a copy)"(Array:Array of User Widget Object References=GetAllEntries(self=Get EntryBox_TimeWindows()), Dimension 1=@n6) -> (Output:User Widget Object Reference)
   n8 = GetNumEntries(self=Get EntryBox_TimeWindows()) -> (ReturnValue=0)
   n9 = Branch(execute=[@n5.then, @n4.else], Condition=AND Boolean(A="integer < integer"(A=@n6, B=@n8), B=IsValid(Object=@n7)))
   > SetNavigationRuleExplicit!(execute, self:Widget Object Reference=@n1, Direction:EUINavigation Enum=Down, InWidget:Widget Object Reference=@n7) -> (then)
   > SetNavigationRuleExplicit(self=@n7, Direction=Up, InWidget=@n1)
   n10 = Branch(execute=@n9.else, Condition=IsVisible(self=Get Background_AddWindow()))
   n11 = Get Button_AddWindow()
   > SetNavigationRuleExplicit(self=@n1, Direction=Down, InWidget=@n11)
   > SetNavigationRuleExplicit(self=@n11, Direction=Up, InWidget=@n1)
   SetNavigationRuleBase(execute=@n10.else, self=@n1, Direction=Down, Rule=Escape)
  ]}
 GetBottomFocusableWidget {nodes [
   GetBottomFocusableWidget() {is=title K2Node_FunctionEntry}
   n1 = IsVisible(self=Get Background_AddWindow())
   n2 = GetNumEntries(self=Get EntryBox_TimeWindows()) -> (ReturnValue=0)
   > n3 = Branch(Condition=AND Boolean(A="integer > integer"(A=@n2), B=NOT Boolean(A=@n1)))
   n4 = Get EntryBox_TimeWindows()
   n5 = GetNumEntries(self=@n4) -> (ReturnValue=0)
   > Return Node!(execute, ReturnValue:Widget Object Reference="Get (a copy)"(Array=GetAllEntries(self=@n4), Dimension 1=int - int(A=@n5)))
   Return Node(execute=@n3.else, ReturnValue=Get Button_AddWindow())
  ]}
 GetPrimaryGamepadFocusWidget {nodes [
   GetPrimaryGamepadFocusWidget() {is=title K2Node_FunctionEntry}
   > n1 = Branch(Condition="integer > integer"(A=Get LastTimeWindowSelectedToReturnFocusTo() -> (LastTimeWindowSelectedToReturnFocusTo:int=0), B=-1))
   n2 = Get EntryBox_TimeWindows()
   > n3 = Branch(Condition="integer > integer"(A=GetNumEntries(self=@n2) -> (ReturnValue=0), B=Get LastTimeWindowSelectedToReturnFocusTo() -> (LastTimeWindowSelectedToReturnFocusTo=0)))
   n4 = Get EntryBox_TimeWindows()
   > Return Node(ReturnValue="Get (a copy)"(Array=GetAllEntries(self=@n4), Dimension 1=Get LastTimeWindowSelectedToReturnFocusTo() -> (LastTimeWindowSelectedToReturnFocusTo=0)))
   Return Node(execute=@n3.else, ReturnValue=Get Button_AddWindow())
   n5 = GetOwningListView!(UserListEntry:User List Entry Interface) -> (ReturnValue:List View Base Object Reference)
   n6 = Cast To SettingsPanel_ContentControls!(execute=@n1.else, Object=GetOuterObject!(Object:Object Reference=GetOuterObject(Object=@n5)) -> (ReturnValue:Object Reference)) -> (then, CastFailed, AsSettings Panel Content Controls:Settings Panel Content Controls Object Reference) {is=title K2Node_DynamicCast}
   n7 = Get LastSelectedIndex!(self:Settings Panel Content Controls Object Reference=@n6) -> (LastSelectedIndex:int=0)
   n8 = Get SettingIndex() -> (SettingIndex:int=0)
   n9 = Get SettingIndex() -> (SettingIndex=0)
   n10 = Get CurrentSelectedIndex!(self:Settings Panel Content Controls Object Reference=@n6) -> (CurrentSelectedIndex:int=0)
   > n11 = Branch(Condition=OR Boolean(A="integer < integer"(A=@n10, B=@n9), B="integer < integer"(A=@n7, B=@n8)))
   n12 = GetTopFocusableWidget!(execute=[@n11.then, @n16.else], self) -> (then, ReturnValue:Widget Object Reference)
   > Return Node(ReturnValue=@n12)
   n13 = Get CurrentSelectedIndex(self=@n6) -> (CurrentSelectedIndex=0)
   n14 = Get SettingIndex() -> (SettingIndex=0)
   n15 = Get SettingIndex() -> (SettingIndex=0)
   n16 = Branch(execute=@n11.else, Condition=OR Boolean(A="integer > integer"(A=@n7, B=@n15), B="integer > integer"(A=@n13, B=@n14)))
   n17 = GetBottomFocusableWidget!(execute=[@n16.then, @n6.CastFailed], self) -> (then, ReturnValue:Widget Object Reference) {is=title K2Node_CallFunction}
   > Return Node(ReturnValue=@n17)
  ]}
 GetTopFocusableWidget {nodes [
   GetTopFocusableWidget() {is=title K2Node_FunctionEntry}
   n1 = Get EntryBox_TimeWindows()
   > n2 = Branch(Condition="integer > integer"(A=GetNumEntries(self=@n1) -> (ReturnValue=0)))
   n3 = Get EntryBox_TimeWindows()
   > Return Node(ReturnValue="Get (a copy)"(Array=GetAllEntries(self=@n3)))
   Return Node(execute=@n2.else, ReturnValue=Get Button_AddWindow())
  ]}
]
variables [SettingIndex {type=integer}]
widgets {
 info {
  inheritedWidgetCount=3
  namedSlotCount=1
  parentClass=/Script/FortniteUI.FortSettingsListEntrySetting_TimeWindows
  rootWidgetClass=&VerticalBox
  widgetCount=7
 }
 widgets [
  VerticalBox_0 {widgetClassPath=&VerticalBox, widget=VerticalBox_0}
  EntryBox_TimeWindows {
   widgetClassPath=&DynamicEntryBox
   bInherited=true
   bIsVariable=true
   parent=VerticalBox_0
   widget=EntryBox_TimeWindows
   properties {entryBoxType=Vertical, entryWidgetClass=&/SettingsListEntry_TimeWindow.SettingsListEntry_TimeWindow_C}
   slot {class=VerticalBoxSlot}
  }
  Background_AddWindow {
   widgetClassPath=&/SettingEntryBackground
   bIsVariable=true
   parent=VerticalBox_0
   widget=Background_AddWindow
   properties {bIsFocusable=true}
   slot {class=VerticalBoxSlot}
  }
  FortResponsivePanel_0 {widgetClassPath=/Script/FortniteUI.FortResponsivePanel, namedSlotHost=Background_AddWindow, widget=FortResponsivePanel_0}
  Button_AddWindow {
   widgetClassPath=&IconTextButton
   bIsVariable=true
   parent=FortResponsivePanel_0
   widget=Button_AddWindow
   properties {
    buttonText=+ Add Another Window
    contentAlignment=HAlign_Center
    onButtonBaseClicked=(&self:WidgetTree.Button_AddWindow.HandleOnButtonClicked)
    onButtonBaseDoubleClicked=(&self:WidgetTree.Button_AddWindow.HandleOnButtonDoubleClicked)
    onButtonBaseHovered=(&self:WidgetTree.Button_AddWindow.HandleOnButtonHovered)
    onButtonBaseUnhovered=(&self:WidgetTree.Button_AddWindow.HandleOnButtonUnhovered)
    onSelectedChangedBase=(&self:WidgetTree.Button_AddWindow.HandleOnSelectedChanged)
    style=/FortUILibrary/Foundation/Buttons/ButtonStyles/Athena/Athena_ButtonStyle_BlueMenuButton_Settings.Athena_ButtonStyle_BlueMenuButton_Settings_C
    touchMethod=PreciseTap
   }
   slot {class=FortResponsivePanelSlot}
  }
  Text_SettingName {widgetClassPath=&CommonTextBlock, bInherited=true, widget=null}
  Background {widgetClassPath=&UserWidget, bInherited=true, widget=null}
 ]
}
