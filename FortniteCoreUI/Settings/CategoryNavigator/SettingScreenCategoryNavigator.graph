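# Settings-screen category navigator widget, a SettingCategoryNavigator subclass holding a legacy category list and a UIKit side navigation in a visibility switcher.
# Category updates rebuild the UIKit tabs and a dev-name-to-setting map, which translates tab selection, click and gamepad focus back to parent callbacks.
# Collapse and expand, gated by an aspect-ratio check, animate the SizeBox width and update the side nav.
# Compact UI adds left padding and clears the width override.
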WidgetBlueprint /FortniteCoreUI/Settings/CategoryNavigator/SettingScreenCategoryNavigator
animations [CollapseNavigatorAnim {
  binding_count=1
  length_seconds=0.2333
  object=&self:CollapseNavigatorAnim
  sequence {
   bindings [SizeBox_All {
     id=#1
     possessed_class=SizeBox
     tracks [{
       type=MovieSceneFloatTrack
       display_name=Width Override
       sections [{
         type=MovieSceneFloatSection
         end_frame=7
         start_frame=0
         channels [None {type=MovieSceneScriptingFloatChannel, keys [[0,380],[3,220],[7,44]]}]
        }]
      }]
    }]
   display_rate {numerator=30}
  }
 }]
graphs [
 AddCategory {nodes [
   n1 = AddCategory!() -> (then, SettingCategory:Fort Setting Object Reference) {is=title K2Node_FunctionEntry}
   n2 = String To Name(InString=GetDevName!(self:Fort Setting Object Reference=@n1) -> (ReturnValue:String))
   > Add!(
    execute
    TargetMap:Map of Names to Fort Setting Object References=Get SettingCategoryMap() -> (SettingCategoryMap:Map of Names to Fort Setting Object References)
    Key:"Name (by ref)"=@n2
    Value:"Fort Setting Object Reference (by ref)"=@n1
   ) -> (then) {is=title K2Node_CallFunction}
   > AddSideNavigationTab!(
    execute
    self:WBP UIKit Side Navigation Base Object Reference=Get SideNav_UIKit() -> (SideNav_UIKit:WBP UIKit Side Navigation Object Reference)
    TabViewModel:MVVM Base Viewmodel Object Reference
    TabID:Name=@n2
    Index:int=-1
   ) -> (then)
   > n3 = Cast To WBP_UIKit_SideNavigation_Button!(
    execute
    Object=GetTabButtonBaseByID!(self:Common Tab List Widget Base Object Reference=Get SideNav_UIKit(), TabNameID:Name=@n2) -> (ReturnValue:Common Button Base Object Reference)
   ) -> (then, CastFailed, AsWBP UIKit Side Navigation Button:WBP UIKit Side Navigation Button Object Reference) {is=title K2Node_DynamicCast}
   > SetText!(
    execute
    self:WBP UIKit Button Generic Base Object Reference=@n3
    Text:Text=GetDisplayName!(self:Fort Setting Object Reference=Get SettingCategory() -> (SettingCategory:Fort Setting Object Reference)) -> (ReturnValue:"Text (by ref)")
   ) -> (then)
   > Bind Event to On Focused(self:Common Button Base Object Reference=@n3, Delegate=Create Event()) {is=title K2Node_AddDelegate}
  ]}
 EventGraph {nodes [
   n1 = "On Side Navigation Selection Changed (SideNav_UIKit)"!() -> (OutputDelegate, then, TabID:Name=None, InnerTabID:Name=None, TabViewModel:MVVM Base Viewmodel Object Reference) {is=title K2Node_ComponentBoundEvent}
   n2 = Find(TargetMap:Map of Names to Fort Setting Object References=Get SettingCategoryMap(), Key:"Name (by ref)"=@n1.TabID) -> (Value:Fort Setting Object Reference)
   > Branch(Condition=@n2.ReturnValue)
   > OnSettingCategorySelectionChanged_UIKit!(execute, self:Setting Category Navigator Object Reference, SettingCategory:Fort Setting Object Reference=@n2) -> (then)
   n3 = "On Side Navigation Tab Clicked (SideNav_UIKit)"!() -> (OutputDelegate, then, TabID:Name=None, IsTabContainerExpanded:bool, TabViewModel:MVVM Base Viewmodel Object Reference) {is=title K2Node_ComponentBoundEvent}
   n4 = Find(TargetMap=Get SettingCategoryMap(), Key=@n3.TabID)
   > Branch(Condition=@n4.ReturnValue)
   > OnSettingCategoryClicked_UIKit!(execute, self:Setting Category Navigator Object Reference, SettingCategory:Fort Setting Object Reference=@n4) -> (then)
   OnSideNavigationButtonFocused_UIKit!() -> (OutputDelegate, then, Button:Common Button Base Object Reference) {is=title K2Node_CustomEvent}
   > Branch(Condition=GetIsUsingGamepad!(self) -> (IsUsingGamepad:bool))
   n5 = Find(
    TargetMap=Get SettingCategoryMap()
    Key=GetSelectedTabId!(self:Common Tab List Widget Base Object Reference=Get SideNav_UIKit()) -> (ReturnValue:Name=None)
   )
   > Branch(Condition=@n5.ReturnValue)
   > OnSettingCategoryScrolledIntoView_UIKit!(execute, self:Setting Category Navigator Object Reference, SettingCategory:Fort Setting Object Reference=@n5) -> (then)
   Event Construct()
   > Branch(Condition=IsUsingCompactUI())
   > SetPadding!(
    execute
    self:Overlay Slot Object Reference=SlotAsOverlaySlot(Widget=Get SizeBox_All() -> (SizeBox_All:Size Box Object Reference))
    InPadding_Left:float=56
    InPadding_Top:float=0
    InPadding_Right:float=0
    InPadding_Bottom:float=0
   ) -> (then)
   > DelayUntilNextFrame!(execute) -> (then)
   > ClearWidthOverride(self:Size Box Object Reference=Get SizeBox_All())
   Event BP_NavigateToIndex_UIKit!() -> (OutputDelegate, then, Index:int=0) {is=title K2Node_Event}
   > SelectTabByIndex!(execute, self:UIKit Tab Group Object Reference=Get SideNav_UIKit(), Index:int=0, bSuppressClickFeedback:bool=true) -> (then, ReturnValue:bool)
   n6 = Event BP_CollapseNavigator!() -> (OutputDelegate, then, bInstant:bool) {is=title K2Node_Event}
   n7 = CompareToScreenAspectRatio!(Comparison:EAspectRatioComparison Enum=LessThan, AspectRatio:EAspectRatio Enum=Ratio_19_5_9) -> (ReturnValue:bool)
   > Branch(Condition=@n7)
   > SetIsCollapsed!(execute, self, IsCollapsed:bool=true, Instant:bool=@n6.bInstant) -> (then)
   n8 = Get CollapseNavigatorAnim() -> (CollapseNavigatorAnim:Widget Animation Object Reference)
   > PlayAnimationForward(InAnimation=@n8)
   > Branch(Condition=@n6.bInstant)
   > SetAnimationCurrentTime!(execute, self:User Widget Object Reference, InAnimation:Widget Animation Object Reference=@n8, InTime:float=GetEndTime(self=@n8)) -> (then)
   n9 = Event BP_ExpandNavigator!() -> (OutputDelegate, then, bInstant:bool) {is=title K2Node_Event}
   > Branch(Condition=@n7)
   > SetIsCollapsed(IsCollapsed=false, Instant=@n9.bInstant)
   > PlayAnimationReverse(InAnimation=@n8)
   > Branch(Condition=@n9.bInstant)
   > SetAnimationCurrentTime(InAnimation=@n8, InTime=GetStartTime!(self:Widget Animation Object Reference=@n8) -> (ReturnValue:float))
   n10 = Event BP_UpdateCategories_UIKit!() -> (OutputDelegate, then, SettingCategories:Array of Fort Setting Object References) {is=title K2Node_Event}
   n11 = Length!(TargetMap:Map of Names to Fort Setting Object References=Get SettingCategoryMap()) -> (ReturnValue:int=0) {is=title K2Node_CallFunction}
   > n12 = Branch(Condition=AND Boolean(A=GetIsUsingGamepad(), B="integer > integer"(A=@n11)))
   > SetFocus()
   n13 = Get SideNav_UIKit()
   > SetVisibility(self=@n13, InVisibility=HitTestInvisible)
   > Delay(Duration=0.1)
   > n14 = SetVisibility(self=@n13, InVisibility=Visible)
   BP_ClearCategories_UIKit(execute=[@n14.then, @n12.else], self:Setting Category Navigator Object Reference)
   > n15 = For Each Loop(Array:Array of Fort Setting Object References=@n10.SettingCategories) -> (Array Element:Fort Setting Object Reference)
   > AddCategory!(execute, self, SettingCategory:Fort Setting Object Reference=@n15) -> (then) {is=title K2Node_CallFunction}
   SetIsCollapsed(execute=@n15.Completed, IsCollapsed=Get IsSideNavCollapsed() -> (IsSideNavCollapsed:bool), Instant=true)
   > SelectTabByIndex(self=Get SideNav_UIKit(), Index=0, bSuppressClickFeedback=true)
   > FocusSelectedSideBarButton()
   Event BP_ClearCategories_UIKit() {is=title K2Node_Event}
   > Clear!(execute, TargetMap:Map of Names to Fort Setting Object References=Get SettingCategoryMap()) -> (then) {is=title K2Node_CallFunction}
   > ClearSideNavigationTabs(self:WBP UIKit Side Navigation Base Object Reference=Get SideNav_UIKit())
   n16 = Event BP_SetSelectedCategory_UIKit!() -> (OutputDelegate, then, SettingCategory:Fort Setting Object Reference) {is=title K2Node_Event}
   > SelectSideNavigationTab!(
    execute
    self:WBP UIKit Side Navigation Base Object Reference=Get SideNav_UIKit()
    TabID:Name=String To Name(InString=GetDevName(self=@n16.SettingCategory))
    InnerTabID:Name=None
    bSuppressClickFeedback:bool=true
   ) -> (then, Success:bool)
  ]}
 FocusSelectedSideBarButton {nodes [
   FocusSelectedSideBarButton() {is=title K2Node_FunctionEntry}
   > SetFocus(self=GetSideNavigationSelectedButtonBase!(self:WBP UIKit Side Navigation Base Object Reference=Get SideNav_UIKit()) -> (Button:Common Button Base Object Reference))
  ]}
 GetIsUsingGamepad {nodes [
   GetIsUsingGamepad() {is=title K2Node_FunctionEntry}
   n1 = Get CommonInputSubsystem()
   > Return Node!(execute, IsUsingGamepad:bool="Equal (Enum)"(A=GetCurrentInputType(self=@n1), B=Gamepad))
  ]}
 OnFocusReceived {nodes [
   OnFocusReceived!() -> (then, MyGeometry:Geometry Structure, InFocusEvent:Focus Event Structure) {is=title K2Node_FunctionEntry}
   > FocusSelectedSideBarButton() {is=title K2Node_CallFunction}
   > Return Node!(execute, ReturnValue:Event Reply Structure=Handled())
  ]}
 SetIsCollapsed {nodes [
   n1 = SetIsCollapsed!() -> (then, IsCollapsed:bool=~, Instant:bool=~) {is=title K2Node_FunctionEntry}
   > Set IsSideNavCollapsed(IsSideNavCollapsed:bool=@n1) -> (Output_Get:bool)
   > Branch(Condition=IsUsingUIKitSideNav!(self:Setting Category Navigator Object Reference) -> (ReturnValue:bool))
   > SetIsCollapsed!(
    execute
    self:WBP UIKit Side Navigation Base Object Reference=Get SideNav_UIKit()
    IsCollapsed=Get IsCollapsed() -> (IsCollapsed:bool)
    SkipAnimation:bool=Get Instant() -> (Instant:bool)
   ) -> (then) {is=title K2Node_CallFunction}
  ]}
]
variables [
 IsSideNavCollapsed {type=boolean}
 SettingCategoryMap {type="map<FortSetting>", default {}}
]
widgets {
 info {inheritedWidgetCount=2, parentClass=/Script/FortniteUI.SettingCategoryNavigator, rootWidgetClass=&Overlay, widgetCount=8}
 widgets [
  Overlay_1 {widgetClassPath=&Overlay, widget=Overlay_1}
  SizeBox_All {
   widgetClassPath=&SizeBox
   bIsVariable=true
   parent=Overlay_1
   widget=SizeBox_All
   properties {widthOverride=380}
   slot {class=OverlaySlot}
  }
  Switcher_NavComponent {
   widgetClassPath=&CommonVisibilitySwitcher
   bInherited=true
   bIsVariable=true
   parent=SizeBox_All
   widget=Switcher_NavComponent
   slot {class=SizeBoxSlot}
  }
  Border_Background {
   widgetClassPath=&Border
   parent=Switcher_NavComponent
   widget=Border_Background
   properties {
    background {
     bIsDynamicallyLoaded=false
     drawAs=Image
     imageType=NoImage
     mirroring=NoMirror
     resourceName=None
     resourceObject=null
     tiling=NoTile
     imageSize=<32 32>
     margin=<ltrb 0 0 0 0>
     outlineSettings {
      bUseBrushTransparency=false
      roundingType=FixedRadius
      width=0
      color {colorUseRule=UseColor_Specified, specifiedColor=<rgba 0 0 0 0>}
      cornerRadii=<16 16 16 16>
     }
     tintColor {colorUseRule=UseColor_Specified, specifiedColor=<rgba 0 0 0 0>}
     uVRegion {bIsValid=false, max=<0 0>, min=<0 0>}
    }
    brushColor=<rgba 1 1 1 0.4>
   }
   slot {class=CommonVisibilitySwitcherSlot}
  }
  Overlay_0 {widgetClassPath=&Overlay, parent=Border_Background, widget=Overlay_0, slot {class=BorderSlot}}
  UIKit_Backplate {
   widgetClassPath=&WBP_UIKit_Backplate
   parent=Overlay_0
   widget=UIKit_Backplate
   properties {brightness=Dark, cornerRadius=Interactive Component}
   slot {class=OverlaySlot, properties {horizontalAlignment=HAlign_Fill, verticalAlignment=VAlign_Fill}}
  }
  ListView_Categories {
   widgetClassPath=&CommonListView
   bInherited=true
   bIsVariable=true
   parent=Overlay_0
   widget=ListView_Categories
   properties {
    bReturnFocusToSelection=true
    entryWidgetClass=&/SettingsListEntry_Category.SettingsListEntry_Category_C
   }
   slot {class=OverlaySlot, properties {horizontalAlignment=HAlign_Fill, padding=<ltrb 0 16 0 16>}}
  }
  SideNav_UIKit {
   widgetClassPath=/FortUILibrary/UIKit/Tabs/WBP_UIKit_SideNavigation
   bIsVariable=true
   parent=Switcher_NavComponent
   widget=SideNav_UIKit
   properties {scrollBarVisibility=Visible, tabGroupHorizontalAlignment=HAlign_Fill}
   slot {class=CommonVisibilitySwitcherSlot}
  }
 ]
}
