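# Settings list row widget, derived from FortSettingsListEntry_Setting, that stands in for a setting with no editor.
# A responsive panel inside the entry background holds the setting name beside a centered "No Editor Found" label.
# Mouse enter and leave play the OnHover animation forward and in reverse, skipped for touch input, shifting the name text from translucent white to dark navy.
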WidgetBlueprint /FortniteCoreUI/Settings/SettingsListEntry_Missing
animations [OnHover {
  binding_count=1
  length_seconds=0.1
  object=&self:OnHover
  sequence {bindings [Text_SettingName {
     id=#1
     possessed_class=CommonTextBlock
     tracks [{
       type=MovieSceneColorTrack
       display_name=Color and Opacity
       sections [{
         type=MovieSceneColorSection
         end_frame=2
         start_frame=0
         channels [
          Color.R {type=MovieSceneScriptingFloatChannel, keys [[0,1],[2,0.0009]]}
          Color.G {type=MovieSceneScriptingFloatChannel, keys [[0,1],[2,0.006]]}
          Color.B {type=MovieSceneScriptingFloatChannel, keys [[0,1],[2,0.0395]]}
          Color.A {type=MovieSceneScriptingFloatChannel, keys [[0,0.5],[2,0.4]]}
         ]
        }]
      }]
    }]}
 }]
graphs [EventGraph {nodes [
   Event OnMouseEnter()
   n1 = IsUsingTouch(PlayerController=GetOwningPlayer())
   > n2 = Branch(Condition=@n1)
   n3 = Get OnHover() -> (OnHover:Widget Animation Object Reference)
   PlayAnimationForward(execute=@n2.else, InAnimation=@n3)
   Event OnMouseLeave!() -> (OutputDelegate, then, MouseEvent:"Pointer Event Structure (by ref)") {is=title K2Node_Event}
   > n4 = Branch(Condition=@n1)
   PlayAnimationReverse(execute=@n4.else, InAnimation=@n3)
  ]}]
variables []
widgets {
 info {
  inheritedWidgetCount=2
  namedSlotCount=1
  parentClass=/Script/FortniteUI.FortSettingsListEntry_Setting
  rootWidgetClass=&/SettingEntryBackground.SettingEntryBackground_C
  widgetCount=6
 }
 widgets [
  Background {widgetClassPath=&/SettingEntryBackground, bInherited=true, bIsVariable=true, widget=Background}
  FortResponsivePanel_0 {widgetClassPath=/Script/FortniteUI.FortResponsivePanel, namedSlotHost=Background, widget=FortResponsivePanel_0}
  SizeBox_0 {
   widgetClassPath=&SizeBox
   parent=FortResponsivePanel_0
   widget=SizeBox_0
   properties {bOverride_MinDesiredHeight=true, minDesiredHeight=50}
   slot {class=FortResponsivePanelSlot}
  }
  Text_SettingName {
   widgetClassPath=&CommonTextBlock
   bInherited=true
   bIsVariable=true
   parent=SizeBox_0
   widget=Text_SettingName
   properties {
    mobileFontSizeMultiplier=1.13
    scrollStyle=&TextScrollStyle-Base_C
    text="{Setting Name}"
    textTransformPolicy=ToUpper
    colorAndOpacity {colorUseRule=UseColor_Specified, specifiedColor=<rgba 1 1 1 0.5>}
    font {
     bForceMonospaced=false
     bMaterialIsStencil=false
     fontMaterial=null
     fontObject=&BurbankSmall
     letterSpacing=0
     monospacedWidth=1
     size=16
     skewAmount=0
     typefaceFontName=Black
     outlineSettings {
      bApplyOutlineToDropShadows=false
      bMiteredCorners=false
      bSeparateFillAlpha=false
      outlineMaterial=null
      outlineSize=0
      outlineColor=<rgba 0 0 0 1>
     }
    }
    margin=<ltrb 40 0 0 0>
    shadowOffset=<0 0>
   }
   slot {class=SizeBoxSlot, properties {verticalAlignment=VAlign_Center}}
  }
  SizeBox {
   widgetClassPath=&SizeBox
   parent=FortResponsivePanel_0
   widget=SizeBox
   properties {bOverride_MinDesiredHeight=true, minDesiredHeight=50}
   slot {class=FortResponsivePanelSlot}
  }
  Text_SettingName_1 {
   widgetClassPath=&CommonTextBlock
   bIsVariable=true
   parent=SizeBox
   widget=Text_SettingName_1
   properties {
    justification=Center
    scrollStyle=&TextScrollStyle-Base_C
    text=No Editor Found
    textTransformPolicy=ToUpper
    font {
     bForceMonospaced=false
     bMaterialIsStencil=false
     fontMaterial=null
     fontObject=&BurbankSmall
     letterSpacing=0
     monospacedWidth=1
     size=18
     skewAmount=0
     typefaceFontName=Black
     outlineSettings {
      bApplyOutlineToDropShadows=false
      bMiteredCorners=false
      bSeparateFillAlpha=false
      outlineMaterial=null
      outlineSize=0
      outlineColor=<rgba 0 0 0 1>
     }
    }
    margin=<ltrb 10 0 0 0>
    shadowOffset=<0 0>
   }
   slot {class=SizeBoxSlot, properties {verticalAlignment=VAlign_Center}}
  }
 ]
}
