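# Categorized settings screen widget, based on FortCategorizedSettingsPanel, that fills a UIKit screen layout with a category navigator, settings list and scrollable details view.
# On Construct it registers the details ScrollBox as the scroll widget and picks the secondary content fill from screen aspect ratio.
# SetMobileStyle overrides that fill for compact UI and adjusts padding on narrower aspect ratios.
# The layout then receives the fill and refreshes its content panels and safe zone.
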WidgetBlueprint /FortniteCoreUI/Settings/CategorizedSettingsPanel
graphs [
 EventGraph {nodes [
   Event Construct()
   > SetScrollWidget!(execute, ToScroll:Widget Object Reference=Get ScrollBox() -> (ScrollBox:Scroll Box Object Reference)) -> (then)
   > Set SecondaryContentFillValue(SecondaryContentFillValue:double=SelectFloat(
     A=0.5
     B=0.8
     bPickA=CompareToScreenAspectRatio!(Comparison:EAspectRatioComparison Enum=GreaterThanOrEqualTo, AspectRatio:EAspectRatio Enum=Ratio_19_5_9) -> (ReturnValue:bool)
    )) -> (Output_Get:double)
   > SetMobileStyle() {is=title K2Node_CallFunction}
   n1 = Get WBP_UIKit_ScreenLayout() -> (WBP_UIKit_ScreenLayout:WBP UIKit Screen Layout Object Reference)
   > Set SecondaryContent_Fill!(
    execute
    SecondaryContent_Fill:double=Get SecondaryContentFillValue() -> (SecondaryContentFillValue:double)
    self:WBP UIKit Screen Layout Object Reference=@n1
   ) -> (then, Output_Get:double)
   > DisplayContentPanels(self:WBP UIKit Screen Layout Object Reference=@n1)
   > DisplaySafeZone(self:WBP UIKit Screen Layout Object Reference=@n1)
  ]}
 SetMobileStyle {nodes [
   SetMobileStyle() {is=title K2Node_FunctionEntry}
   > Branch(Condition=IsUsingCompactUI())
   n1 = CompareToScreenAspectRatio(Comparison=LessThan, AspectRatio=Ratio_23_16)
   > Set SecondaryContentFillValue(SecondaryContentFillValue=SelectFloat(A=0.9, B=0.8, bPickA=@n1))
   > Branch(Condition=@n1)
   n2 = Get WBP_UIKit_ScreenLayout()
   > Set Mobile_SafeZone_ContentPadding!(
    execute
    Mobile_SafeZone_ContentPadding:Margin Structure=Make Margin(Left=16, Top=24, Right=16, Bottom=120)
    self:WBP UIKit Screen Layout Object Reference=@n2
   ) -> (then, Output_Get:Margin Structure)
   n3 = Get!(self:WBP UIKit Screen Layout Object Reference=@n2) -> (
    PrimaryContent_Padding_Left:float=~
    PrimaryContent_Padding_Top:float=~
    PrimaryContent_Padding_Right:float=~
    PrimaryContent_Padding_Bottom:float=~
   )
   > Set PrimaryContent_Padding!(
    execute
    PrimaryContent_Padding:Margin Structure=Make Margin(Left=@n3, Top=@n3.PrimaryContent_Padding_Top, Right=24, Bottom=@n3.PrimaryContent_Padding_Bottom)
    self:WBP UIKit Screen Layout Object Reference=@n2
   ) -> (then, Output_Get:Margin Structure)
  ]}
]
variables [SecondaryContentFillValue {type=number, default=0.8}]
widgets {
 info {
  inheritedWidgetCount=3
  namedSlotCount=3
  parentClass=/Script/FortniteUI.FortCategorizedSettingsPanel
  rootWidgetClass=/FortUILibrary/UIKit/Layout/WBP_UIKit_ScreenLayout.WBP_UIKit_ScreenLayout_C
  widgetCount=6
 }
 widgets [
  WBP_UIKit_ScreenLayout {
   widgetClassPath=/FortUILibrary/UIKit/Layout/WBP_UIKit_ScreenLayout
   bIsVariable=true
   widget=WBP_UIKit_ScreenLayout
   properties {
    bApplySafeZone=false
    mobile_SafeZone_TopNavPadding=88
    safeZone_TopNavPadding=84
    secondaryContent_Fill=0.8
    mobile_SafeZone_ContentPadding=<ltrb 48 24 48 120>
    primaryContent_Padding=<ltrb 0 0 64 80>
    safeZone_ContentPadding=<ltrb 80 32 80 64>
    secondaryContent_Padding=<ltrb 0 0 0 64>
   }
  }
  VerticalBox {widgetClassPath=&VerticalBox, bIsVariable=true, namedSlotHost=WBP_UIKit_ScreenLayout, widget=VerticalBox}
  ListView_Settings {
   widgetClassPath=/Script/FortniteUI.FortSettingsListView
   bInherited=true
   bIsVariable=true
   parent=VerticalBox
   widget=ListView_Settings
   properties {
    bReturnFocusToSelection=true
    clipping=Inherit
    entryWidgetClass=&/SettingsListEntry_Missing.SettingsListEntry_Missing_C
    horizontalEntrySpacing=4
    navigation=&self:WidgetTree.ListView_Settings.WidgetNavigation_0
    numDesignerPreviewEntries=13
    verticalEntrySpacing=24
    visualData=/FortniteCoreUI/Settings/SettingEntries/SettingsRegistryVisuals
   }
   slot {class=VerticalBoxSlot, properties {size {sizeRule=Fill, value=1}}}
  }
  ScrollBox {widgetClassPath=&ScrollBox, bIsVariable=true, namedSlotHost=WBP_UIKit_ScreenLayout, widget=ScrollBox}
  Details_Settings {
   widgetClassPath=/FortniteCoreUI/Settings/SettingsDetails/GameSettingsDetailView
   bInherited=true
   bIsVariable=true
   parent=ScrollBox
   widget=Details_Settings
   properties {visualData=&SettingsRegistryVisuals}
   slot {class=ScrollBoxSlot, properties {size {sizeRule=Fill, value=1}}}
  }
  Navigator_SettingCategories {
   widgetClassPath=/FortniteCoreUI/Settings/CategoryNavigator/SettingScreenCategoryNavigator
   bInherited=true
   bIsVariable=true
   namedSlotHost=WBP_UIKit_ScreenLayout
   widget=Navigator_SettingCategories
   properties {desiredFocusWidget {widgetName=ListView_Categories}}
  }
 ]
}
