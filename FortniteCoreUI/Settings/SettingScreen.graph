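# SettingScreen is the Fortnite settings screen widget, derived from FortSettingsScreen, with top tabs above a page switcher holding the inherited settings panel.
# Tab creation titles buttons from view models, and selection forwards to HandleTabSelected.
# On init, runtime options set top safe-zone padding, then mobile games restyle tabs and pad all sides.
# SetSelectedTopTabID selects only a valid, different tab, suppressing click feedback.
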WidgetBlueprint /FortniteCoreUI/Settings/SettingScreen
graphs [
 EventGraph {nodes [
   n1 = "On Tab Button Creation (WBP_Settings_Tabs_TopBar)"!() -> (OutputDelegate, then, TabId:Name=None, TabButton:Common Button Base Object Reference) {is=title K2Node_ComponentBoundEvent}
   > n2 = Cast To WBP_Settings_Button_Tab_TopNavBar!(execute, Object=@n1.TabButton) -> (then, CastFailed, AsWBP Settings Button Tab Top Nav Bar:WBP Settings Button Tab Top Nav Bar Object Reference) {is=title K2Node_DynamicCast}
   n3 = Get WBP_Settings_Tabs_TopBar() -> (WBP_Settings_Tabs_TopBar:WBP Settings Tabs Top Bar Object Reference)
   > SetTabTitleText!(
    execute
    self:WBP Settings Button Tab Top Nav Bar Object Reference=@n2
    InTabTitleText:Text=Get Label!(self:UI Kit Selectable Button View Model Object Reference=GetTabViewModelByID!(self:UIKit Tab Group Object Reference=@n3, TabNameID:Name=@n1.TabId) -> (ReturnValue:UI Kit Tab Group Button View Model Object Reference)) -> (Label:Text)
   ) -> (then)
   > SetText!(execute, self:WBP UIKit Button Generic Base Object Reference=@n2, Text:Text=GetEmptyText!() -> (ReturnValue:Text)) -> (then)
   n4 = "On Tab Selected (WBP_Settings_Tabs_TopBar)"!() -> (OutputDelegate, then, TabId:Name=None) {is=title K2Node_ComponentBoundEvent}
   > HandleTabSelected!(execute, self:Fort Settings Screen Object Reference, TabId:Name=@n4.TabId) -> (then)
   Event OnInitialized()
   > Set PadTop!(
    execute
    PadTop:bool=Get bEnableTopSafeZone!(self:Fort Runtime Options Object Reference=GetRuntimeOptions()) -> (bEnableTopSafeZone:bool)
    self:Safe Zone Object Reference=Get SafeZone() -> (SafeZone:Safe Zone Object Reference)
   ) -> (then, Output_Get:bool)
   > SetMobileStyle() {is=title K2Node_CallFunction}
  ]}
 SetMobileStyle {nodes [
   SetMobileStyle() {is=title K2Node_FunctionEntry}
   > Branch(Condition=IsMobileGame())
   n1 = Get WBP_Settings_Tabs_TopBar()
   > Set IconSize!(execute, IconSize:Vector 2D Structure=(X=72,Y=72), self:WBP UIKit Tabs Base Object Reference=@n1) -> (then, Output_Get:Vector 2D Structure)
   > Set UseBackplate!(execute, UseBackplate:bool=false, self:WBP UIKit Tabs Base Object Reference=@n1) -> (then, Output_Get:bool)
   > Set MinButtonHeight!(execute, MinButtonHeight:int=90, self:WBP UIKit Tabs Base Object Reference=@n1) -> (then, Output_Get:int=0)
   n2 = SlotAsVerticalBoxSlot(Widget=@n1)
   > SetHorizontalAlignment(self:Vertical Box Slot Object Reference=@n2, InHorizontalAlignment=HAlign_Left)
   > SetVerticalAlignment(self:Vertical Box Slot Object Reference=@n2, InVerticalAlignment=VAlign_Top)
   > SetPadding!(
    execute
    self:Vertical Box Slot Object Reference=@n2
    InPadding_Left:float=48
    InPadding_Top:float=8
    InPadding_Right:float=0
    InPadding_Bottom:float=0
   ) -> (then)
   > SetPadding(
    self:Border Slot Object Reference=SlotAsBorderSlot!(Widget:Widget Object Reference=Get Overlay() -> (Overlay:Overlay Object Reference)) -> (ReturnValue:Border Slot Object Reference)
    InPadding_Left=0
    InPadding_Top=0
    InPadding_Right=0
    InPadding_Bottom=0
   )
   > SetSidesToPad!(
    execute
    self:Safe Zone Object Reference=Get SafeZone()
    InPadLeft:bool=true
    InPadRight:bool=true
    InPadTop:bool=true
    InPadBottom:bool=true
   ) -> (then)
  ]}
 SetSelectedTopTabID {nodes [
   n1 = SetSelectedTopTabID!() -> (then, TabID:Name) {is=title K2Node_FunctionEntry}
   n2 = Get WBP_Settings_Tabs_TopBar()
   n3 = GetSelectedTabId!(self:Common Tab List Widget Base Object Reference=@n2) -> (ReturnValue:Name=None)
   > Branch(Condition=AND Boolean(A="Not Equal (Name)"(A=@n1) -> (ReturnValue=~), B="Not Equal (Name)"(A=@n1, B=@n3) -> (ReturnValue=~)))
   > SelectTabByID!(
    execute
    self:Common Tab List Widget Base Object Reference=@n2
    TabNameID:Name=Get TabID() -> (TabID:Name=None)
    bSuppressClickFeedback:bool=true
   ) -> (then, ReturnValue:bool)
  ]}
]
variables []
widgets {
 info {inheritedWidgetCount=2, parentClass=/Script/FortniteUI.FortSettingsScreen, rootWidgetClass=&Border, widgetCount=11}
 widgets [
  Border_25 {
   widgetClassPath=&Border
   widget=Border_25
   properties {
    navigation=&self:WidgetTree.Border_25.WidgetNavigation_0
    background {
     bIsDynamicallyLoaded=false
     drawAs=Image
     imageType=FullColor
     mirroring=NoMirror
     resourceName=None
     resourceObject=/FortniteCoreUI/Settings/Materials/MI_UI_Settings_NebulousBackground
     tiling=NoTile
     imageSize=<8589936640 0>
     margin=<ltrb 0 0 0 0>
     outlineSettings {
      bUseBrushTransparency=false
      roundingType=HalfHeightRadius
      width=0
      color {colorUseRule=UseColor_Specified, specifiedColor=<rgba 0 0 0 0>}
      cornerRadii=<0 0 0 0>
     }
     tintColor {colorUseRule=UseColor_Specified, specifiedColor=<rgba 1 1 1 1>}
     uVRegion {bIsValid=false, max=<0 0>, min=<0 0>}
    }
   }
  }
  Overlay {widgetClassPath=&Overlay, bIsVariable=true, parent=Border_25, widget=Overlay, slot {class=BorderSlot}}
  WBP_CaptureForPostBufferUpdate {
   widgetClassPath=&WBP_CaptureForPostBufferUpdate
   bIsVariable=true
   parent=Overlay
   slot {class=OverlaySlot, properties {horizontalAlignment=HAlign_Fill, verticalAlignment=VAlign_Fill}}
  }
  SafeZone {
   widgetClassPath=&SafeZone
   bIsVariable=true
   parent=Overlay
   widget=SafeZone
   properties {padBottom=false, padTop=false}
   slot {class=OverlaySlot, properties {horizontalAlignment=HAlign_Fill, verticalAlignment=VAlign_Fill}}
  }
  VerticalBox_39 {widgetClassPath=&VerticalBox, parent=SafeZone, widget=VerticalBox_39, slot {class=SafeZoneSlot}}
  WBP_Settings_Tabs_TopBar {
   widgetClassPath=/FortniteCoreUI/Settings/Tabs/WBP_Settings_Tabs_TopBar
   bIsVariable=true
   parent=VerticalBox_39
   widget=WBP_Settings_Tabs_TopBar
   properties {
    buttonVerticalAlignment=VAlign_Center
    tabSpacing=8
    localTabsViewModels [
     &self:WidgetTree.WBP_Settings_Tabs_TopBar.UIKitTabGroupButtonViewModel_0
     &self:WidgetTree.WBP_Settings_Tabs_TopBar.UIKitTabGroupButtonViewModel_1
    ]
   }
   slot {class=VerticalBoxSlot, properties {horizontalAlignment=HAlign_Center, padding=<ltrb 0 15 0 0>}}
  }
  Switcher_SettingPages {
   widgetClassPath=&CommonActivatableWidgetSwitcher
   bIsVariable=true
   parent=VerticalBox_39
   widget=Switcher_SettingPages
   slot {class=VerticalBoxSlot, properties {size {sizeRule=Fill, value=1}}}
  }
  Settings_Panel {
   widgetClassPath=&/CategorizedSettingsPanel
   bInherited=true
   bIsVariable=true
   parent=Switcher_SettingPages
   widget=Settings_Panel
   properties {visibility=Visible, desiredFocusWidget {widgetName=Navigator_SettingCategories}}
   slot {class=WidgetSwitcherSlot}
  }
  windowchromebuttons {
   widgetClassPath=/Game/UI/Foundation/Shell/windowchromebuttons
   parent=Overlay
   widget=windowchromebuttons
   slot {class=OverlaySlot, properties {horizontalAlignment=HAlign_Fill}}
  }
  SaveProgressLayer {
   widgetClassPath=&NamedSlot
   bInherited=true
   bIsVariable=true
   parent=Overlay
   widget=SaveProgressLayer
   slot {class=OverlaySlot, properties {horizontalAlignment=HAlign_Fill, verticalAlignment=VAlign_Fill}}
  }
  PluginSpecificUI {
   widgetClassPath=&NamedSlot
   bIsVariable=true
   parent=Overlay
   widget=PluginSpecificUI
   slot {class=OverlaySlot, properties {horizontalAlignment=HAlign_Fill, verticalAlignment=VAlign_Fill}}
  }
 ]
}
