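# Settings panel widget for Fortnite's core UI, a layout-only subclass of the native FortSettingsPanel with an empty event graph.
# A safe-zoned, material-backed border holds a settings list beside a scrollable details view.
# The inherited ListView_Settings and Details_Settings widgets share GameRegistryVisuals data.
# The list's default entry class is SettingsListEntry_Missing, so actual row widgets appear to come from that visual data.
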WidgetBlueprint /FortniteCoreUI/Settings/SettingsPanel
graphs [EventGraph {links [], nodes []}]
variables []
widgets {
 info {inheritedWidgetCount=2, parentClass=/Script/FortniteUI.FortSettingsPanel, rootWidgetClass=&Border, widgetCount=10}
 widgets [
  Border_Background {
   widgetClassPath=&Border
   widget=Border_Background
   properties {
    background {
     bIsDynamicallyLoaded=false
     drawAs=Image
     imageType=NoImage
     mirroring=NoMirror
     resourceName=None
     resourceObject=/FortniteCoreUI/Settings/Art/M_UI_SettingsPanelBG
     tiling=NoTile
     imageSize=<0 0>
     margin=<ltrb 0 0 0 0>
     outlineSettings {
      bUseBrushTransparency=false
      roundingType=HalfHeightRadius
      width=0
      color {colorUseRule=UseColor_Specified, specifiedColor=<rgba 0 0 0 0>}
      cornerRadii=<0 0 0 0>
     }
     tintColor {colorUseRule=UseColor_Specified, specifiedColor=<rgba 1 1 1 1>}
     uVRegion {bIsValid=false, max=<0 0>, min=<0 0>}
    }
    padding=<ltrb 0 0 0 0>
   }
  }
  SafeZone {
   widgetClassPath=&SafeZone
   parent=Border_Background
   widget=SafeZone
   slot {class=BorderSlot, properties {padding=<ltrb 0 0 0 0>}}
  }
  VerticalBox_2 {widgetClassPath=&VerticalBox, parent=SafeZone, widget=VerticalBox_2, slot {class=SafeZoneSlot}}
  Spacer_Top {widgetClassPath=&Spacer, parent=VerticalBox_2, widget=Spacer_Top, properties {size=<1 16>}, slot {class=VerticalBoxSlot}}
  HorizontalBox_2 {
   widgetClassPath=&HorizontalBox
   parent=VerticalBox_2
   widget=HorizontalBox_2
   slot {class=VerticalBoxSlot, properties {size {sizeRule=Fill, value=1}}}
  }
  VerticalBox_0 {
   widgetClassPath=&VerticalBox
   parent=HorizontalBox_2
   widget=VerticalBox_0
   slot {class=HorizontalBoxSlot, properties {size {sizeRule=Fill, value=1}}}
  }
  ListView_Settings {
   widgetClassPath=/Script/FortniteUI.FortSettingsListView
   bInherited=true
   bIsVariable=true
   parent=VerticalBox_0
   widget=ListView_Settings
   properties {
    entryWidgetClass=&/SettingsListEntry_Missing.SettingsListEntry_Missing_C
    horizontalEntrySpacing=4
    verticalEntrySpacing=4
    visualData=&/GameRegistryVisuals
   }
   slot {class=VerticalBoxSlot, properties {padding=<ltrb 10 7 0 0>, size {sizeRule=Fill, value=1}}}
  }
  ScrollBox_0 {
   widgetClassPath=&ScrollBox
   parent=HorizontalBox_2
   widget=ScrollBox_0
   slot {class=HorizontalBoxSlot, properties {padding=<ltrb 70 0 30 64>, size {sizeRule=Fill, value=0.9}}}
  }
  Details_Settings {
   widgetClassPath=&/GameSettingsDetailView
   bInherited=true
   bIsVariable=true
   parent=ScrollBox_0
   widget=Details_Settings
   properties {visualData=&GameRegistryVisuals}
   slot {class=ScrollBoxSlot}
  }
  Spacer_1 {widgetClassPath=&Spacer, parent=VerticalBox_2, widget=Spacer_1, properties {size=<1 64>}, slot {class=VerticalBoxSlot}}
 ]
}
