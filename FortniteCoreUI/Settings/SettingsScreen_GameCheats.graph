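# Cheats settings screen widget, built on the native CheatSettingsScreen class, that shows a titled game cheats settings page.
# A title bar strip holds the "Cheats" heading, a legacy tab list with bumper-style input actions, window chrome buttons and a touch-only close button.
# An inherited settings panel fills the body below.
# On activation it sets the window title bar to overlay mode; the OnSettingsApplied function is empty.
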WidgetBlueprint /FortniteCoreUI/Settings/SettingsScreen_GameCheats
graphs [
 EventGraph {nodes [
   Event On Activated()
   > SetWindowTitleBarState!(
    execute
    TitleBarContent:Widget Object Reference
    Mode:EWindowTitleBarMode Enum=Overlay
    bTitleBarDragEnabled:bool=false
    bWindowButtonsVisible:bool=false
    bTitleBarVisible:bool=true
   ) -> (then)
  ]}
 OnSettingsApplied {nodes [OnSettingsApplied() {is=title K2Node_FunctionEntry}]}
]
variables [OnSettingsApplied {}]
widgets {
 info {inheritedWidgetCount=2, parentClass=/Script/FortniteUI.CheatSettingsScreen, rootWidgetClass=&Border, widgetCount=17}
 widgets [
  Border_0 {widgetClassPath=&Border, widget=Border_0, properties {brushColor=<rgba 0.0437 0.0999 0.3916 1>, padding=<ltrb 0 0 0 0>}}
  HorizontalBox_0 {
   widgetClassPath=&HorizontalBox
   parent=Border_0
   widget=HorizontalBox_0
   slot {class=BorderSlot, properties {padding=<ltrb 0 0 0 0>}}
  }
  VerticalBox_0 {
   widgetClassPath=&VerticalBox
   parent=HorizontalBox_0
   widget=VerticalBox_0
   slot {class=HorizontalBoxSlot, properties {size {sizeRule=Fill, value=1}}}
  }
  CommonBorder_0 {
   widgetClassPath=&CommonBorder
   parent=VerticalBox_0
   widget=CommonBorder_0
   properties {
    style=/FortUILibrary/Styles/Border/Border-SolidBG-ShellBlue.Border-SolidBG-ShellBlue_C
    background {
     bIsDynamicallyLoaded=false
     drawAs=Box
     imageType=NoImage
     mirroring=NoMirror
     resourceName=None
     resourceObject=/FortUILibrary/Styles/Border/Materials/M_UI_Border-SolidBG-ShellBlue
     tiling=NoTile
     imageSize=<240 76>
     margin=<ltrb 0.01 0.05 0.98 0.95>
     outlineSettings {
      bUseBrushTransparency=false
      roundingType=HalfHeightRadius
      width=0
      color {colorUseRule=UseColor_Specified, specifiedColor=<rgba 0 0 0 0>}
      cornerRadii=<0 0 0 0>
     }
     tintColor {colorUseRule=UseColor_Specified, specifiedColor=<rgba 1 1 1 1>}
     uVRegion {bIsValid=false, max=<0 0>, min=<0 0>}
    }
   }
   slot {class=VerticalBoxSlot, properties {size {sizeRule=Automatic, value=0.2}}}
  }
  WindowTitleBarArea_0 {widgetClassPath=/Script/UMG.WindowTitleBarArea, parent=CommonBorder_0, widget=WindowTitleBarArea_0, slot {class=BorderSlot}}
  SafeZone_0 {
   widgetClassPath=&SafeZone
   parent=WindowTitleBarArea_0
   widget=SafeZone_0
   properties {padBottom=false}
   slot {class=WindowTitleBarAreaSlot}
  }
  HorizontalBox_1 {
   widgetClassPath=&HorizontalBox
   parent=SafeZone_0
   widget=HorizontalBox_1
   properties {flowDirectionPreference=Culture}
   slot {class=SafeZoneSlot}
  }
  HorizontalBox_2 {
   widgetClassPath=&HorizontalBox
   parent=HorizontalBox_1
   widget=HorizontalBox_2
   slot {class=HorizontalBoxSlot, properties {verticalAlignment=VAlign_Center, size {sizeRule=Fill, value=1}}}
  }
  HorizontalBox_5 {
   widgetClassPath=&HorizontalBox
   parent=HorizontalBox_2
   widget=HorizontalBox_5
   slot {class=HorizontalBoxSlot, properties {size {sizeRule=Fill, value=1}}}
  }
  SectionImage {
   widgetClassPath=&Image
   bIsVariable=true
   parent=HorizontalBox_5
   widget=SectionImage
   properties {
    visibility=Collapsed
    brush {
     bIsDynamicallyLoaded=false
     drawAs=Image
     imageType=NoImage
     mirroring=NoMirror
     resourceName=None
     resourceObject=/Game/UI/Foundation/Textures/Icons/Social/T-Icon-Config-64
     tiling=NoTile
     imageSize=<52 52>
     margin=<ltrb 0 0 0 0>
     outlineSettings {
      bUseBrushTransparency=false
      roundingType=HalfHeightRadius
      width=0
      color {colorUseRule=UseColor_Specified, specifiedColor=<rgba 0 0 0 0>}
      cornerRadii=<0 0 0 0>
     }
     tintColor {colorUseRule=UseColor_Specified, specifiedColor=<rgba 1 1 1 1>}
     uVRegion {bIsValid=false, max=<0 0>, min=<0 0>}
    }
   }
   slot {class=HorizontalBoxSlot, properties {verticalAlignment=VAlign_Center}}
  }
  SizeBox_HeadingLeto {
   widgetClassPath=&SizeBox
   parent=HorizontalBox_5
   widget=SizeBox_HeadingLeto
   slot {class=HorizontalBoxSlot, properties {verticalAlignment=VAlign_Center, padding=<ltrb 10 0 0 0>}}
  }
  Text_Title {
   widgetClassPath=&CommonTextBlock
   bIsVariable=true
   parent=SizeBox_HeadingLeto
   widget=Text_Title
   properties {
    text=Cheats
    textTransformPolicy=ToUpper
    font {
     bForceMonospaced=false
     bMaterialIsStencil=false
     fontMaterial=null
     fontObject=&BurbankSmall
     letterSpacing=0
     monospacedWidth=1
     size=22
     skewAmount=0
     typefaceFontName=Black
     outlineSettings {
      bApplyOutlineToDropShadows=false
      bMiteredCorners=false
      bSeparateFillAlpha=false
      outlineMaterial=null
      outlineSize=0
      outlineColor=<rgba 0 0 0 1>
     }
    }
    shadowOffset=<0 0>
   }
   slot {class=SizeBoxSlot}
  }
  SettingsTabs {
   widgetClassPath=/Game/UI/Foundation/TabbedView/HorizontalTabList_Legacy
   bIsVariable=true
   parent=HorizontalBox_2
   widget=SettingsTabs
   properties {
    bAutoListenForInput=true
    buttonStyle=/FortUILibrary/Foundation/Buttons/ButtonStyles/Tab/ButtonStyle-Tab-Manage.ButtonStyle-Tab-Manage_C
    flowDirectionPreference=LeftToRight
    forceTabLabelsCollapsed=true
    onTabButtonCreation=(&self:WidgetTree.SettingsTabs.HandleOnTabButtonCreated)
    onTabButtonRemoval=(&self:WidgetTree.SettingsTabs.HandleOnTabButtonRemoved)
    useButtonStyle=true
    nextTabInputActionData {dataTable=&InputActionDataTable, rowName=MainTabRight}
    previousTabInputActionData {dataTable=&InputActionDataTable, rowName=MainTabLeft}
   }
   slot {class=HorizontalBoxSlot, properties {horizontalAlignment=HAlign_Center, size {sizeRule=Fill, value=1}}}
  }
  windowchromebuttons {
   widgetClassPath=/Game/UI/Foundation/Shell/windowchromebuttons
   parent=HorizontalBox_2
   widget=windowchromebuttons
   slot {class=HorizontalBoxSlot, properties {size {sizeRule=Fill, value=1}}}
  }
  VisibilityBorder_Touch {
   widgetClassPath=&CommonHardwareVisibilityBorder
   parent=HorizontalBox_2
   widget=VisibilityBorder_Touch
   properties {visibilityQuery {
     autoDescription=" ANY( Input.Touch )"
     tokenStreamVersion=0
     userDescription=""
     queryTokenStream [0,1,1,1,0]
     tagDictionary [{tagName=Input.Touch}]
    }}
   slot {class=HorizontalBoxSlot}
  }
  Button_CloseTouch {
   widgetClassPath=/FortUILibrary/Foundation/Buttons/CloseButton
   bInherited=true
   bIsVariable=true
   parent=VisibilityBorder_Touch
   widget=Button_CloseTouch
   properties {
    onButtonBaseClicked=(&self:WidgetTree.Button_CloseTouch.HandleOnButtonClicked)
    onButtonBaseDoubleClicked=(&self:WidgetTree.Button_CloseTouch.HandleOnButtonDoubleClicked)
    onButtonBaseHovered=(&self:WidgetTree.Button_CloseTouch.HandleOnButtonHovered)
    onButtonBaseUnhovered=(&self:WidgetTree.Button_CloseTouch.HandleOnButtonUnhovered)
    onSelectedChangedBase=(&self:WidgetTree.Button_CloseTouch.HandleOnSelectedChanged)
   }
   slot {class=BorderSlot, properties {padding=<ltrb 0 0 0 0>}}
  }
  Settings_Panel {
   widgetClassPath=&/SettingsPanel
   bInherited=true
   bIsVariable=true
   parent=VerticalBox_0
   widget=Settings_Panel
   slot {class=VerticalBoxSlot, properties {size {sizeRule=Fill, value=1}}}
  }
 ]
}
