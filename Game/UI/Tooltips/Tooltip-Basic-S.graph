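# Small CommonUI tooltip widget showing an optional icon, header and body text in a size-limited, shadowed border.
# On Construct it applies any valid header and body text styles, sets both texts, and collapses empty parts, or the whole widget when all are empty.
# A spacer indents the body when an icon exists.
# The divider's visibility is set three times, so it ultimately follows only whether body text exists.
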WidgetBlueprint /Game/UI/Tooltips/Tooltip-Basic-S
graphs [
 EventGraph {nodes [
   Event Construct()
   > n1 = Sequence!(execute) -> (then_0, then_1, then_2:Exec)
   n2 = Get Header Style() -> (Header Style:Common Text Style Class Reference)
   > Branch(Condition=IsValidClass!(Class:Object Class Reference=@n2) -> (ReturnValue:bool))
   > SetStyle!(
    execute
    self:Common Text Object Reference=Get Text-Header() -> (Text-Header:Common Text Object Reference)
    InStyle:Common Text Style Class Reference=@n2
   ) -> (then)
   n3 = Sequence!(execute=@n1.then_2) -> (then_0, then_1, then_2, then_3:Exec, then_4:Exec, then_5:Exec)
   n4 = IsValid(Object=Get Icon Brush() -> (Icon Brush:Slate Brush Asset Object Reference))
   > SetVisibility(
    self=[Get Icon-Header() -> (Icon-Header:Image Object Reference), Get Spacer-Icon() -> (Spacer-Icon:Spacer Object Reference)]
    InVisibility=Select(Option 0=Collapsed, Option 1=Visible, Index=@n4)
   )
   n5 = NOT Boolean(A=TextIsEmpty(InText=Get HeaderText() -> (HeaderText:Text)))
   SetVisibility(
    execute=@n3.then_1
    self=["Get <hr>"() -> ("<hr>":Image Object Reference), Get Text-Header()]
    InVisibility=Select(Option 0=Collapsed, Option 1=Visible, Index=@n5)
   )
   n6 = OR Boolean(A=@n4, B=@n5)
   SetVisibility(
    execute=@n3.then_2
    self=[Get Header HBox() -> (Header HBox:Horizontal Box Object Reference), "Get <hr>"()]
    InVisibility=Select(Option 0=Collapsed, Option 1=Visible, Index=@n6)
   )
   n7 = NOT Boolean(A=TextIsEmpty(InText=Get BodyText() -> (BodyText:Text)))
   SetVisibility(
    execute=@n3.then_3
    self=["Get <hr>"(), Get Text-Body() -> (Text-Body:Common Text Object Reference), Get Body HBox() -> (Body HBox:Horizontal Box Object Reference)]
    InVisibility=Select(Option 0=Collapsed, Option 1=Visible, Index=@n7)
   )
   SetVisibility(execute=@n3.then_4, InVisibility=Select(Option 0=Collapsed, Option 1=Visible, Index=OR Boolean(A=@n6, B=@n7)))
   "SetText (Text)"(execute=@n3.then_5, self=Get Text-Header(), InText=Get HeaderText())
   > "SetText (Text)"(self=Get Text-Body(), InText=Get BodyText())
   n8 = Get Body Style() -> (Body Style:Common Text Style Class Reference)
   Branch(execute=@n1.then_1, Condition=IsValidClass(Class=@n8))
   > SetStyle(self=Get Text-Body(), InStyle=@n8)
  ]}
 Get Body Text {nodes [
   Get Body Text() {is=title K2Node_FunctionEntry}
   > Return Node!(execute, ReturnValue:Text=Get BodyText())
  ]}
 Get Header Text {nodes [
   Get Header Text() {is=title K2Node_FunctionEntry}
   > Return Node(ReturnValue=Get HeaderText())
  ]}
 Get Icon Brush {nodes [
   Get Icon Brush() {is=title K2Node_FunctionEntry}
   > Return Node(ReturnValue:Slate Brush Structure=MakeBrushFromAsset!(BrushAsset:Slate Brush Asset Object Reference=Get Icon Brush() {is=title K2Node_VariableGet}) -> (ReturnValue:Slate Brush Structure))
  ]}
]
variables [
 Body Style {type=Class, default=/FortUILibrary/Styles/Typography/Foundation/TextStyle-Base-XS.TextStyle-Base-XS_C}
 BodyText {type=string, default=""}
 Header Style {type=Class, default=/FortUILibrary/Styles/Typography/Foundation/TextStyle-Base-XS-40pc.TextStyle-Base-XS-40pc_C}
 HeaderText {type=string, default=""}
 Icon Brush {type=SlateBrushAsset, default=null}
]
widgets {
 info {inheritedWidgetCount=0, parentClass=&CommonUserWidget, rootWidgetClass=&SizeBox, widgetCount=12}
 widgets [
  SizeBox_0 {
   widgetClassPath=&SizeBox
   widget=SizeBox_0
   properties {
    bOverride_MaxDesiredWidth=true
    bOverride_MinDesiredHeight=true
    bOverride_MinDesiredWidth=true
    maxDesiredWidth=580
    minDesiredHeight=32
    minDesiredWidth=128
   }
  }
  Overlay_4 {widgetClassPath=&Overlay, parent=SizeBox_0, widget=Overlay_4, slot {class=SizeBoxSlot}}
  CommonBorder_268 {
   widgetClassPath=&CommonBorder
   parent=Overlay_4
   widget=CommonBorder_268
   properties {
    style=/FortUILibrary/Styles/Border/Border-SolidBG-ToolTipShadow.Border-SolidBG-ToolTipShadow_C
    background {
     bIsDynamicallyLoaded=false
     drawAs=Box
     imageType=NoImage
     mirroring=NoMirror
     resourceName=None
     resourceObject=/FortUILibrary/Styles/Border/Materials/M-Border-ToolTip
     tiling=NoTile
     imageSize=<32 32>
     margin=<ltrb 0.5 0.5 0.5 0.5>
     outlineSettings {
      bUseBrushTransparency=false
      roundingType=HalfHeightRadius
      width=0
      color {colorUseRule=UseColor_Specified, specifiedColor=<rgba 0 0 0 0>}
      cornerRadii=<0 0 0 0>
     }
     tintColor {colorUseRule=UseColor_Specified, specifiedColor=<rgba 1 1 1 1>}
     uVRegion {bIsValid=false, max=<0 0>, min=<0 0>}
    }
    renderTransform {angle=0, scale=<1 1>, shear=<0 0>, translation=<6 8>}
   }
   slot {class=OverlaySlot, properties {horizontalAlignment=HAlign_Fill, verticalAlignment=VAlign_Fill}}
  }
  CommonBorder_0 {
   widgetClassPath=&CommonBorder
   parent=Overlay_4
   widget=CommonBorder_0
   properties {style=/FortUILibrary/Styles/Border/Border-SolidBG-ToolTip.Border-SolidBG-ToolTip_C, padding=<ltrb 12 12 12 12>}
   slot {class=OverlaySlot, properties {horizontalAlignment=HAlign_Fill, verticalAlignment=VAlign_Fill}}
  }
  VerticalBox_509 {
   widgetClassPath=&VerticalBox
   parent=CommonBorder_0
   widget=VerticalBox_509
   slot {class=BorderSlot, properties {padding=<ltrb 12 12 12 12>}}
  }
  Header HBox {
   widgetClassPath=&HorizontalBox
   bIsVariable=true
   parent=VerticalBox_509
   widget=Header HBox
   slot {class=VerticalBoxSlot, properties {padding=<ltrb 0 0 0 4>}}
  }
  Icon-Header {
   widgetClassPath=&Image
   bIsVariable=true
   parent=Header HBox
   widget=Icon-Header
   properties {brush {
     bIsDynamicallyLoaded=false
     drawAs=Image
     imageType=NoImage
     mirroring=NoMirror
     resourceName=None
     resourceObject=/Game/UI/Icons/Icon-BasicFrame-S
     tiling=NoTile
     imageSize=<24 24>
     margin=<ltrb 0 0 0 0>
     outlineSettings {
      bUseBrushTransparency=false
      roundingType=HalfHeightRadius
      width=0
      color {colorUseRule=UseColor_Specified, specifiedColor=<rgba 0 0 0 0>}
      cornerRadii=<0 0 0 0>
     }
     tintColor {colorUseRule=UseColor_Specified, specifiedColor=<rgba 1 1 1 1>}
     uVRegion {bIsValid=false, max=<0 0>, min=<0 0>}
    }}
   slot {class=HorizontalBoxSlot, properties {horizontalAlignment=HAlign_Left, verticalAlignment=VAlign_Top, padding=<ltrb 0 0 8 0>}}
  }
  Text-Header {
   widgetClassPath=&CommonTextBlock
   bIsVariable=true
   parent=Header HBox
   widget=Text-Header
   properties {
    autoWrapText=true
    style=/FortUILibrary/Styles/Typography/Foundation/TextStyle-Base-S-B.TextStyle-Base-S-B_C
    font {
     bForceMonospaced=false
     bMaterialIsStencil=false
     fontMaterial=null
     fontObject=&NotoSans
     letterSpacing=0
     monospacedWidth=1
     size=18
     skewAmount=0
     typefaceFontName=NotoSans-Bold
     outlineSettings {
      bApplyOutlineToDropShadows=false
      bMiteredCorners=false
      bSeparateFillAlpha=false
      outlineMaterial=null
      outlineSize=0
      outlineColor=<rgba 0 0 0 1>
     }
    }
    shadowOffset=<0 0>
   }
   slot {class=HorizontalBoxSlot, properties {size {sizeRule=Fill, value=1}}}
  }
  "<hr>" {
   widgetClassPath=&Image
   bIsVariable=true
   parent=VerticalBox_509
   widget="<hr>"
   properties {
    brush {
     bIsDynamicallyLoaded=false
     drawAs=Image
     imageType=NoImage
     mirroring=NoMirror
     resourceName=None
     resourceObject=null
     tiling=Horizontal
     imageSize=<6 2>
     margin=<ltrb 0 0 0 0>
     outlineSettings {
      bUseBrushTransparency=false
      roundingType=HalfHeightRadius
      width=0
      color {colorUseRule=UseColor_Specified, specifiedColor=<rgba 0 0 0 0>}
      cornerRadii=<0 0 0 0>
     }
     tintColor {colorUseRule=UseColor_Specified, specifiedColor=<rgba 1 1 1 1>}
     uVRegion {bIsValid=false, max=<0 0>, min=<0 0>}
    }
    colorAndOpacity=<rgba 1 1 1 0.2>
   }
   slot {class=VerticalBoxSlot, properties {padding=<ltrb 0 0 0 4>}}
  }
  Body HBox {
   widgetClassPath=&HorizontalBox
   bIsVariable=true
   parent=VerticalBox_509
   widget=Body HBox
   slot {class=VerticalBoxSlot, properties {size {sizeRule=Fill, value=1}}}
  }
  Spacer-Icon {
   widgetClassPath=&Spacer
   bIsVariable=true
   parent=Body HBox
   widget=Spacer-Icon
   properties {size=<24 1>}
   slot {class=HorizontalBoxSlot, properties {padding=<ltrb 0 0 8 0>}}
  }
  Text-Body {
   widgetClassPath=&CommonTextBlock
   bIsVariable=true
   parent=Body HBox
   widget=Text-Body
   properties {
    autoWrapText=true
    style=/FortUILibrary/Styles/Typography/Foundation/TextStyle-Base-S.TextStyle-Base-S_C
    font {
     bForceMonospaced=false
     bMaterialIsStencil=false
     fontMaterial=null
     fontObject=&NotoSans
     letterSpacing=0
     monospacedWidth=1
     size=18
     skewAmount=0
     typefaceFontName=NotoSans-Regular
     outlineSettings {
      bApplyOutlineToDropShadows=false
      bMiteredCorners=false
      bSeparateFillAlpha=false
      outlineMaterial=null
      outlineSize=0
      outlineColor=<rgba 0 0 0 1>
     }
    }
    shadowOffset=<0 0>
   }
   slot {class=HorizontalBoxSlot, properties {size {sizeRule=Fill, value=1}}}
  }
 ]
}
